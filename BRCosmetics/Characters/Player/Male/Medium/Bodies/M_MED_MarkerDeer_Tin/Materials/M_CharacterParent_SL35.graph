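# Parent character master material for Super Levels S35 styles, emitting masked glowing inner lines and Fresnel rim.
# Shared PreFX output passes through a galaxy function and switchable hair and flyaway-hair functions before PostFX.
# A tinted glow, limited by the inverted PreFX blue mask, is max-combined into emissive.
# Inner lines come from an SDF texture, antialiased with screen derivatives and thickness varying by FOV-compensated distance.
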Material /BRCosmetics/Characters/Player/Male/Medium/Bodies/M_MED_MarkerDeer_Tin/Materials/M_CharacterParent_SL35
graph {
 properties {
  bHasPixelAnimation=true
  bUseMaterialAttributes=true
  bUsedWithMeshDeformer=true
  bUsedWithMorphTargets=true
  bUsedWithSkeletalMesh=true
  pixelDepthOffsetMode=PDOM_Legacy
  shadingModel=MSM_FromMaterialExpression
  usedShadingModels=Default Lit | Subsurface | Clear Coat | Subsurface Profile | Cloth
 }
 nodes [
  n1 = MaterialFunctionCall() {materialFunction=&MF_CharacterParent_2023_PreFX}
  n2 = MaterialFunctionCall!(Material=@n1.Result, TechArtMask=@n1.Result) {materialFunction=&/MF_NewGalaxy_Generic}
  n3 = MaterialFunctionCall!(In=@n2.Result, Mask Input=@n1.M) {materialFunction=/Game/Global/Materials/Character_Parent_2023/Hair_Parent_2023/MaterialFunctions/MF_FN_Hair_2023}
  n4 = BlendMaterialAttributes(A=@n2.Result, B=StaticSwitchParameter(True=@n3.Result, False=@n2.Result) {group=Skin FX, parameterName=UseFNHair2_0}, Alpha=@n3.HairMask)
  n5 = GetMaterialAttributes(None=StaticSwitchParameter(
    True=MaterialFunctionCall!(MaterialInput=@n4) {materialFunction=/Game/Global/Materials/Character_Parent_2023/Hair_Parent_2023/MaterialFunctions/MF_FlyAwayHair_2023}.Result
    False=@n4
   ) {group=Skin FX, parameterName=IsFlyAwayHair?}) {attributeGetTypes [#1]}
  n6 = MaterialFunctionCall(Float3=@n1.M) {materialFunction=&BreakOutFloat3Components}
  n7 = NamedRerouteUsage()
  n8 = TextureSampleParameter2D() {group=SuperLevels S35, parameterName=Fresnel_Mask, samplerType=SAMPLERTYPE_Masks, texture=&T_WhiteMask}
  n9 = ChannelMaskParameter(None=@n8.RGB) {group=SuperLevels S35, parameterName=Fresnel_MaskChannel}
  n10 = Multiply(
   A=NamedRerouteUsage()
   B=LinearInterpolate(A=@n9.R, B=OneMinus(None=@n9.R), Alpha=ScalarParameter() {group=SuperLevels S35, parameterName=Fresnel_MaskInvert})
  )
  n11 = Multiply(
   A=OneMinus(None=NamedRerouteDeclaration(None=@n6.B) {name=MaskMap, nodeColor=<rgba 0 0.9297 1 1>})
   B=Multiply(
    A=Max(A=@n7, B=@n10)
    B=VectorParameter() {group=SuperLevels S35, parameterName=Fresnel_Color, defaultValue=<rgba 0.8388 0.5583 0.0152 1>}.RGB
   )
  )
  MaterialFunctionCall(
   Material=SetMaterialAttributes(MaterialAttributes=@n5.MaterialAttributes, Emissive Color=Max(A=@n5.Emissive Color, B=@n11)) {attributeSetTypes [#1]}
   M=@n1.M
   Pre Skinned Local Position=@n1.M
   Distance Blend=@n1.M
   Fresnel=@n1.M
  ) {materialFunction=&MF_CharacterParent_2023_PostFX}
  n12 = TextureSampleParameter2D() {group=SuperLevels S35, parameterName=GlowSDFMask, samplerType=SAMPLERTYPE_Alpha, texture=/Game/Global/Textures/Default/Blanks/T_BlackAlpha}
  MaterialFunctionCall!(
   ThresholdValue=ScalarParameter() {defaultValue=1, group=SuperLevels S35, parameterName=InnerLines_Threshold}
   ThresholdRange=Add(A=NamedRerouteUsage(), B=ScalarParameter() {defaultValue=0.002, group=SuperLevels S35, parameterName=InnerLines_Range})
   Alpha=NamedRerouteDeclaration(None=Saturate(None=@n12.RGB)) {name=InnerLines_Gradient, nodeColor=<rgba 1 0.9375 0 1>}
   TwoSidedThreshold
  ) {materialFunction=/Engine/Functions/Engine_MaterialFunctions02/Masking/ThresholdWithRange}
  NamedRerouteDeclaration(None=@n6.G) {name=GMap, nodeColor=<rgba 0.0312 0 1 1>}
  NamedRerouteDeclaration(None=@n6.R) {name=AOMap, nodeColor=<rgba 1 0 0.6094 1>}
  n13 = NamedRerouteUsage()
  n14 = Add(A=Abs(None=DDX(Value=@n13)), B=Abs(None=DDY(Value=@n13)))
  n15 = WorldPosition()
  n16 = CameraPositionWS()
  n17 = Add(
   A=Divide(A=Distance(A=@n15.XYZ, B=@n16), B=ScalarParameter() {defaultValue=5000, group=SuperLevels S35, parameterName=InnerLines_MaxDistance}) {constB=5000}
   B=ScalarParameter() {defaultValue=0.3, group=SuperLevels S35, parameterName=InnerLines_MinDistance}
  ) {constB=0.3}
  NamedRerouteDeclaration(None=Multiply(A=Multiply(A=@n14, B=@n17), B=ScalarParameter() {defaultValue=3, group=SuperLevels S35, parameterName=InnerLine_Radius})) {name=InnerLines_Modulation, nodeColor=<rgba 1 0 0.1875 1>}
  n18 = ScalarParameter() {defaultValue=0.4, group=SuperLevels S35, parameterName=InnerLines_LineThickness_Near, sliderMax=1}
  n19 = ScalarParameter() {defaultValue=0.6, group=SuperLevels S35, parameterName=InnerLines_LineThickness_Far, sliderMax=1}
  n20 = MaterialFunctionCall!(
   Fade Offset=ScalarParameter() {defaultValue=50, group=SuperLevels S35, parameterName=InnerLines_FadeStart, sliderMax=200}
   Fade Distance=ScalarParameter() {defaultValue=1000, group=SuperLevels S35, parameterName=InnerLines_FadeDistance, sliderMax=1000}
  ) {materialFunction=/Game/Global/Materials/Character_Parent_2023/Anime_Parent_2023/MaterialFunctions/MFC_DistanceFade_FOV_compensation}
  n21 = TextureSampleParameter2D() {group=SuperLevels S35, parameterName=GlowSDFMask, samplerType=SAMPLERTYPE_Alpha, texture=&T_BlackAlpha}
  n22 = Subtract(A=OneMinus(None=LinearInterpolate(A=@n18, B=@n19, Alpha=@n20.Result)), B=Power(Base=@n21.R) {constExponent=0.4545, desc=gamma correction})
  n23 = AppendVector(A=DDX(Value=@n22), B=DDY(Value=@n22))
  n24 = ScalarParameter() {defaultValue=0.8, group=SuperLevels S35, parameterName=InnerLines_LineHardness, sliderMax=1}
  n25 = Divide(A=@n22, B=Add(A=Length(None=@n23) {is=class MaterialExpressionLength}, B=LinearInterpolate(Alpha=@n24) {constA=0.5, constB=0}) {constB=0})
  NamedRerouteDeclaration(None=Saturate(None=Subtract(B=@n25) {constA=0.5})) {name=InnerLines_OutMask, nodeColor=<rgba 0 0.2734 1 1>}
  n26 = VertexNormalWS()
  n27 = Fresnel(
   ExponentIn=ScalarParameter() {defaultValue=3, group=SuperLevels S35, parameterName=Fresnel_Exponent}
   Normal=LinearInterpolate(
    A=VertexInterpolator(VS=@n26).PS
    B=PixelNormalWS()
    Alpha=ScalarParameter() {defaultValue=1, group=SuperLevels S35, parameterName=Fresnel_PixelNormalBlend}
   ) {constAlpha=0}
  ) {baseReflectFraction=0, exponent=3}
  n28 = ScalarParameter() {defaultValue=0.5, group=SuperLevels S35, parameterName=Fresnel_Contrast}
  NamedRerouteDeclaration(None=Multiply(
    A=MaterialFunctionCall(In=@n27, Contrast=@n28) {materialFunction=&CheapContrast}.Result
    B=ScalarParameter() {defaultValue=1, group=SuperLevels S35, parameterName=Fresnel_Intensity}
   )) {name=Fresnel, nodeColor=<rgba 0.5937 1 0 1>}
  NamedRerouteDeclaration(None=@n2.Alpha Mask) {name=AlphaMask, nodeColor=<rgba 0.875 0 1 1>}
 ]
}
