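# Control Rig for the MarkerDeer Tin player body that corrects deformation and drives costume helper bones as the body moves.
# The forward solve runs a shared DeformRig function, then collar, hose, jacket, fly, gun and suspender solutions.
# These weight bone offsets mostly with spherical pose readers, plus constraints, anim attributes, swing-twist angle remapping and a Sliding flag.
# The hose chain fits a spline whose inner control points are solved from four input points.
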ControlRigBlueprint /BRCosmetics/Characters/Player/Male/Medium/Bodies/M_MED_MarkerDeer_Tin/Meshes/M_MED_MarkerDeer_Tin_CtrlRig
graphs [
 ContainedGraph {nodes [
   RigVMFunctionEntryNode()
   n1 = /Script/RigVM.RigVMFunction_MathFloatDiv(
    A=&RigVMFunction_MathFloatMul(A=RigVMVariableNode() {variable=a}.Value, B=RigVMVariableNode() {variable=e}.Value) {template="Multiply::Execute(in A,in B,out Result)"}.Result
    B=RigVMVariableNode() {variable=d}.Value
   ) {is=struct RigVMUnitNode, template="Divide::Execute(in A,in B,out Result)"}
   n2 = RigVMVariableNode() {variable=b}
   n3 = &RigVMFunction_MathFloatDiv(A=RigVMVariableNode() {variable=a}.Value, B=RigVMVariableNode() {variable=d}.Value) {template="Divide::Execute(in A,in B,out Result)"}
   n4 = RigVMVariableNode() {variable=f}
   n5 = /Script/RigVM.RigVMFunction_MathVectorScale(
    Factor=&RigVMFunction_MathFloatDiv(
     A=1
     B=/Script/RigVM.RigVMFunction_MathFloatSub(B=@n1.Result, A=@n2.Value) {is=struct RigVMUnitNode, template="Subtract::Execute(in A,in B,out Result)"}.Result
    ) {template="Divide::Execute(in A,in B,out Result)"}.Result
    Value=/Script/RigVM.RigVMFunction_MathVectorSub(
     B=&RigVMFunction_MathVectorScale(Factor=@n3.Result, Value=@n4.Value) {template="Scale::Execute(in Value,in Factor,out Result)"}.Result
     A=RigVMVariableNode() {variable=c}.Value
    ) {is=struct RigVMUnitNode, template="Subtract::Execute(in A,in B,out Result)"}.Result
   ) {is=struct RigVMUnitNode, template="Scale::Execute(in Value,in Factor,out Result)"}
   n6 = RigVMVariableNode() {variable=a}
   n7 = &RigVMFunction_MathVectorScale(Value=@n5.Result, Factor=RigVMVariableNode() {variable=b}.Value) {template="Scale::Execute(in Value,in Factor,out Result)"}
   n8 = RigVMVariableNode() {variable=c}
   RigVMFunctionReturnNode(
    j=@n5.Result
    i=&RigVMFunction_MathVectorScale(
     Factor=&RigVMFunction_MathFloatDiv(A=1, B=@n6.Value) {template="Divide::Execute(in A,in B,out Result)"}.Result
     Value=&RigVMFunction_MathVectorSub(B=@n7.Result, A=@n8.Value) {template="Subtract::Execute(in A,in B,out Result)"}.Result
    ) {template="Scale::Execute(in Value,in Factor,out Result)"}.Result
   )
  ]}
 ContainedGraph {nodes [
   n1 = RigVMFunctionEntryNode()
   n2 = &RigVMFunction_MathFloatSub(A=1, B=@n1.t) {template="Subtract::Execute(in A,in B,out Result)"}
   n3 = &RigVMFunction_MathFloatMul(A=@n2.Result, B=@n2.Result) {template="Multiply::Execute(in A,in B,out Result)"}
   RigVMFunctionReturnNode(Result=&RigVMFunction_MathFloatMul(B=3, A=&RigVMFunction_MathFloatMul(A=@n1.t, B=@n3.Result) {template="Multiply::Execute(in A,in B,out Result)"}.Result) {template="Multiply::Execute(in A,in B,out Result)"}.Result)
  ]}
 ContainedGraph {nodes [
   n1 = RigVMFunctionEntryNode()
   n2 = &RigVMFunction_MathFloatMul(A=@n1.t, B=@n1.t) {template="Multiply::Execute(in A,in B,out Result)"}
   n3 = &RigVMFunction_MathFloatSub(A=1, B=@n1.t) {template="Subtract::Execute(in A,in B,out Result)"}
   RigVMFunctionReturnNode(Result=&RigVMFunction_MathFloatMul(B=3, A=&RigVMFunction_MathFloatMul(B=@n2.Result, A=@n3.Result) {template="Multiply::Execute(in A,in B,out Result)"}.Result) {template="Multiply::Execute(in A,in B,out Result)"}.Result)
  ]}
 ContainedGraph {nodes [
   n1 = RigVMFunctionEntryNode()
   n2 = &RigVMFunction_MathFloatSub(A=1, B=@n1.t) {template="Subtract::Execute(in A,in B,out Result)"}
   n3 = &RigVMFunction_MathFloatMul(A=@n2.Result, B=@n2.Result) {template="Multiply::Execute(in A,in B,out Result)"}
   RigVMFunctionReturnNode(Result=&RigVMFunction_MathFloatMul(B=3, A=&RigVMFunction_MathFloatMul(A=@n1.t, B=@n3.Result) {template="Multiply::Execute(in A,in B,out Result)"}.Result) {template="Multiply::Execute(in A,in B,out Result)"}.Result)
  ]}
 ContainedGraph {nodes [
   n1 = RigVMFunctionEntryNode()
   n2 = &RigVMFunction_MathFloatMul(A=@n1.t, B=@n1.t) {template="Multiply::Execute(in A,in B,out Result)"}
   n3 = &RigVMFunction_MathFloatSub(A=1, B=@n1.t) {template="Subtract::Execute(in A,in B,out Result)"}
   RigVMFunctionReturnNode(Result=&RigVMFunction_MathFloatMul(B=3, A=&RigVMFunction_MathFloatMul(B=@n2.Result, A=@n3.Result) {template="Multiply::Execute(in A,in B,out Result)"}.Result) {template="Multiply::Execute(in A,in B,out Result)"}.Result)
  ]}
 ContainedGraph {nodes [
   n1 = RigVMFunctionEntryNode()
   RigVMFunctionReturnNode(Result=/Script/RigVM.RigVMFunction_MathFloatPow(B=3, A=&RigVMFunction_MathFloatSub(A=1, B=@n1.t) {template="Subtract::Execute(in A,in B,out Result)"}.Result) {is=struct RigVMUnitNode, template="Power::Execute(in A,in B,out Result)"}.Result)
  ]}
 ContainedGraph {nodes [RigVMFunctionReturnNode(Result=&RigVMFunction_MathFloatPow(B=3, A=RigVMFunctionEntryNode().t) {template="Power::Execute(in A,in B,out Result)"}.Result)]}
 ContainedGraph {nodes [
   n1 = RigVMFunctionEntryNode()
   RigVMFunctionReturnNode(Result=&RigVMFunction_MathFloatPow(B=3, A=&RigVMFunction_MathFloatSub(A=1, B=@n1.t) {template="Subtract::Execute(in A,in B,out Result)"}.Result) {template="Power::Execute(in A,in B,out Result)"}.Result)
  ]}
 ContainedGraph {nodes [RigVMFunctionReturnNode(Result=&RigVMFunction_MathFloatPow(B=3, A=RigVMFunctionEntryNode().t) {template="Power::Execute(in A,in B,out Result)"}.Result)]}
 New Function_ContainedGraph {nodes [
   n1 = RigVMFunctionEntryNode()
   RigVMFunction_Sequence = &RigVMFunction_Sequence(ExecuteContext=@n1.ExecuteContext)
   RigUnit_GetTransform = &RigUnit_GetTransform(Item=(Type=Bone,Name="neck_01"), Space=LocalSpace, bInitial=False)
   RigVMFunction_MathQuaternionSwingTwist = &RigVMFunction_MathQuaternionSwingTwist(TwistAxis=(X=1,Y=0,Z=0), Input=@RigUnit_GetTransform.Transform.Rotation)
   RigUnit_SetRotation = &RigUnit_SetRotation(
    Item=(Type=Bone,Name="dyn_neckNoTwist")
    Weight=1
    bInitial=False
    Value=@RigVMFunction_MathQuaternionSwingTwist.Swing
    ExecutePin=@RigVMFunction_Sequence.A
   )
   > n2 = &RigUnit_SphericalPoseReader(
    ActiveRegionScaleFactors=(PositiveWidth=1,NegativeWidth=3,PositiveHeight=3,NegativeHeight=3)
    ActiveRegionSize=0.1
    Debug=(bDrawDebug=false,bDraw2D=False,bDrawLocalAxes=False,DebugScale=25,DebugSegments=20,DebugThickness=0.25)
    DriverAxis=(X=1,Y=0,Z=0)
    DriverItem=(Type=Bone,Name="head")
    FalloffRegionScaleFactors=(PositiveWidth=1,NegativeWidth=1,PositiveHeight=1,NegativeHeight=1)
    FalloffSize=0.3
    FlipHeightScaling=False
    FlipWidthScaling=False
    OptionalParentItem=(Type=Bone,Name="spine_05")
    RotationOffset=(X=-90,Y=0,Z=5)
   )
   > &RigUnit_ModifyTransforms(
    ItemToModify=((Item=(Type=Bone,Name="dyn_c_collar"),Transform=(Rotation=(X=0,Y=0,Z=-0.0872,W=0.9962),Translation=(X=0,Y=0,Z=0),Scale3D=(X=1,Y=1,Z=1))))
    WeightMinimum=0
    Weight=@n2.OutputParam
   )
   DISPATCH_RigDispatch_GetAnimAttribute = RigVMDispatchNode(BoneName=root, Default=0, Name=deformRig_l_upperarm_up_alpha) {template="DISPATCH_RigDispatch_GetAnimAttribute(in Name,in BoneName,in Default,out Value,out Found)"}
   &RigUnit_ModifyTransforms(
    ItemToModify=((Item=(Type=Bone,Name="dyn_l_collar"),Transform=(Rotation=(X=0,Y=0.1305,Z=0,W=0.9914),Translation=(X=0,Y=0,Z=0),Scale3D=(X=1,Y=1,Z=1))))
    WeightMinimum=0
    Weight=@DISPATCH_RigDispatch_GetAnimAttribute.Value
    ExecutePin=@RigVMFunction_Sequence.B
   )
   > &RigUnit_ModifyTransforms(
    ItemToModify=((Item=(Type=Bone,Name="dyn_r_collar"),Transform=(Rotation=(X=0,Y=0.1736,Z=0,W=0.9848),Translation=(X=0,Y=0,Z=0),Scale3D=(X=1,Y=1,Z=1))))
    WeightMinimum=0
    Weight=RigVMDispatchNode(BoneName=root, Default=0, Name=deformRig_r_upperarm_up_alpha) {template="DISPATCH_RigDispatch_GetAnimAttribute(in Name,in BoneName,in Default,out Value,out Found)"}.Value
   )
   > n3 = &RigUnit_SphericalPoseReader(
    ActiveRegionScaleFactors=(PositiveWidth=0.5,NegativeWidth=6,PositiveHeight=1,NegativeHeight=6)
    ActiveRegionSize=0.1
    Debug=(bDrawDebug=false,DebugScale=25,DebugSegments=20,DebugThickness=0.25)
    DriverAxis=(X=1,Y=0,Z=0)
    DriverItem=(Type=Bone,Name="upperarm_l")
    FalloffRegionScaleFactors=(PositiveWidth=1,NegativeWidth=1,PositiveHeight=1,NegativeHeight=1)
    FalloffSize=0.6
    OptionalParentItem=(Type=Bone)
    RotationOffset=(X=70,Y=-60,Z=7)
   )
   > &RigUnit_ModifyTransforms(
    ItemToModify=((Item=(Type=Bone,Name="dyn_l_collar"),Transform=(Rotation=(X=0,Y=-0.4226,Z=0,W=0.9063),Translation=(X=0,Y=2.5,Z=0),Scale3D=(X=1,Y=1,Z=1))))
    WeightMinimum=0
    Weight=@n3.OutputParam
   )
   > n4 = &RigUnit_SphericalPoseReader(
    ActiveRegionScaleFactors=(PositiveWidth=0.5,NegativeWidth=6,PositiveHeight=0.5,NegativeHeight=6)
    ActiveRegionSize=0.1
    Debug=(bDrawDebug=false,DebugScale=25,DebugSegments=20,DebugThickness=0.25)
    DriverAxis=(X=-1,Y=0,Z=0)
    DriverItem=(Type=Bone,Name="upperarm_r")
    FalloffRegionScaleFactors=(PositiveWidth=1,NegativeWidth=1,PositiveHeight=1,NegativeHeight=1)
    FalloffSize=0.6
    FlipWidthScaling=true
    OptionalParentItem=(Type=Bone)
    RotationOffset=(X=90,Y=210,Z=180)
   )
   > &RigUnit_ModifyTransforms(
    ItemToModify=((Item=(Type=Bone,Name="dyn_r_collar"),Transform=(Rotation=(X=0,Y=-0.1736,Z=0,W=0.9848),Translation=(X=0,Y=-0.5,Z=-0.7),Scale3D=(X=1,Y=1,Z=1))))
    WeightMinimum=0
    Weight=@n4.OutputParam
   )
   > RigVMFunctionReturnNode()
  ]}
 New Function_ContainedGraph {nodes [
   RigVMFunctionEntryNode()
   > RigVMFunction_Sequence = RigVMAggregateNode()
   n1 = &RigUnit_SphericalPoseReader(
    ActiveRegionScaleFactors=(PositiveWidth=3.5,NegativeWidth=1,PositiveHeight=3,NegativeHeight=3)
    ActiveRegionSize=0.1
    Debug=(bDrawDebug=false,bDraw2D=false,DebugScale=25,DebugSegments=20,DebugThickness=0.25)
    DriverAxis=(X=-1,Y=0,Z=0)
    DriverItem=(Type=Bone,Name="thigh_l")
    FalloffRegionScaleFactors=(PositiveWidth=1,NegativeWidth=1,PositiveHeight=1,NegativeHeight=1)
    FalloffSize=0.7
    OptionalParentItem=(Type=Bone,Name="pelvis")
    RotationOffset=(X=110,Y=0,Z=-20)
    ExecutePin=@RigVMFunction_Sequence.A
   )
   > &RigUnit_ModifyTransforms(
    ItemToModify=((Item=(Type=Bone,Name="dyn_hose_upper8_end"),Transform=(Rotation=(X=0,Y=0,Z=0,W=1),Translation=(X=-15,Y=0,Z=0),Scale3D=(X=1,Y=1,Z=1))))
    WeightMinimum=0.1
    Weight=@n1.OutputParam
   )
   > &RigUnit_ModifyTransforms(
    ItemToModify=((Item=(Type=Bone,Name="dyn_hose_upper11_end"),Transform=(Rotation=(X=0,Y=0,Z=0,W=1),Translation=(X=5,Y=0,Z=-5),Scale3D=(X=1,Y=1,Z=1))))
    Weight=1
    WeightMinimum=0.1
   )
   > &RigUnit_ModifyTransforms(
    ItemToModify=((Item=(Type=Bone,Name="dyn_strapHelper_l"),Transform=(Rotation=(X=-0.3738,Y=-0.4606,Z=-0.611,W=0.5242),Translation=(X=10,Y=0,Z=0),Scale3D=(X=1,Y=1,Z=1))))
    WeightMinimum=0.1
    Weight=@n1.OutputParam
   )
   > n2 = &RigUnit_SphericalPoseReader(
    ActiveRegionScaleFactors=(PositiveWidth=3.5,NegativeWidth=1.5,PositiveHeight=3,NegativeHeight=3)
    ActiveRegionSize=0.1
    Debug=(bDrawDebug=false,bDraw2D=false,DebugScale=25,DebugSegments=20,DebugThickness=0.25)
    DriverAxis=(X=1,Y=0,Z=0)
    DriverItem=(Type=Bone,Name="thigh_r")
    FalloffRegionScaleFactors=(PositiveWidth=1,NegativeWidth=1,PositiveHeight=1,NegativeHeight=1)
    FalloffSize=0.6
    OptionalParentItem=(Type=Bone,Name="pelvis")
    RotationOffset=(X=-110,Y=180,Z=-20)
   )
   > &RigUnit_ModifyTransforms(
    ItemToModify=((Item=(Type=Bone,Name="dyn_strapHelper_r"),Transform=(Rotation=(X=-0.0353,Y=-0.3031,Z=-0.577,W=0.7576),Translation=(X=-9,Y=0,Z=0),Scale3D=(X=1,Y=1,Z=1))))
    WeightMinimum=0.1
    Weight=@n2.OutputParam
   )
   > &RigUnit_ModifyTransforms(
    ItemToModify=((Item=(Type=Bone,Name="dyn_hose_upper8_end"),Transform=(Rotation=(X=0,Y=0,Z=0,W=1),Translation=(X=0,Y=0,Z=0),Scale3D=(X=1,Y=1,Z=1))))
    WeightMinimum=0.1
    Weight=@n2.OutputParam
   )
   > &RigUnit_RotationConstraintLocalSpaceOffset(
    AdvancedSettings=(InterpolationType=Average,RotationOrderForFilter=XZY)
    Child=(Type=Bone,Name="dyn_hose_ALT")
    Filter=(bX=True,bY=True,bZ=True)
    Parents=((Item=(Type=Bone,Name="dyn_hose_end_base"),Weight=1),(Item=(Type=Bone,Name="dyn_hose_upper_base"),Weight=1))
    Weight=1
   )
   > /Script/ControlRig.RigUnit_PositionConstraintLocalSpaceOffset(
    Child=(Type=Bone,Name="dyn_hose_ALT")
    Filter=(bX=True,bY=True,bZ=True)
    Parents=((Item=(Type=Bone,Name="dyn_hose_end_base"),Weight=1),(Item=(Type=Bone,Name="dyn_hose_upper_base"),Weight=1))
    Weight=1
    bMaintainOffset=True
   ) {is=struct RigVMUnitNode}
   "Spline four points " = RigVMFunctionReferenceNode(
    SamplesPerSegment=6
    Point 0=&RigUnit_GetTransform(Item=(Type=Control,Name="dyn_hose_upper_end_ctrl"), bInitial=False).Transform.Translation
    Point 1=&RigUnit_GetTransform(Item=(Type=Control,Name="dyn_hose_upper4_end_ctrl"), bInitial=False).Transform.Translation
    Point 2=&RigUnit_GetTransform(Item=(Type=Control,Name="dyn_hose_upper11_end_ctrl"), bInitial=false).Transform.Translation
    Point 3=&RigUnit_GetTransform(Item=(Type=Control,Name="dyn_hose_upper14_end_ctrl"), bInitial=false).Transform.Translation
    ExecuteContext=@RigVMFunction_Sequence.C
   )
   RigUnit_ItemArray = &RigUnit_ItemArray(Items="((Type=Bone,Name=\"dyn_hose_upper\"),(Type=Bone,Name=\"dyn_hose_upper1\"),(Type=Bone,Name=\"dyn_hose_upper2\"),(Type=Bone,Name=\"dyn_hose_upper3\"),(Type=Bone,Name=\"dyn_hose_upper4\"),(Type=Bone,Name=\"dyn_hose_upper5\"),(Type=Bone,Name=\"dyn_hose_upper6\"),(Type=Bone, ...(truncated)")
   > /Script/ControlRigSpline.RigUnit_FitChainToSplineCurveItemArray(
    Alignment=Stretched
    DebugSettings=(bEnabled=False,Scale=1,CurveColor=(R=1,G=1,B=0,A=1),SegmentsColor=(R=1,G=0,B=0,A=1),WorldOffset=(Rotation=(X=0,Y=0,Z=0,W=1),Translation=(X=0,Y=0,Z=0),Scale3D=(X=1,Y=1,Z=1)))
    Maximum=1
    Minimum=0
    PoleVectorPosition=(X=0,Y=0,Z=0)
    PrimaryAxis=(X=1,Y=0,Z=0)
    RotationEaseType=Linear
    Rotations=()
    SamplingPrecision=12
    SecondaryAxis=(X=0,Y=0,Z=0)
    Weight=1
    bPropagateToChildren=true
    Items=@RigUnit_ItemArray.Items
    Spline=@"Spline four points ".Spline
   ) {is=struct RigVMUnitNode}
   > &RigUnit_ParentConstraint(
    AdvancedSettings=(InterpolationType=Average,RotationOrderForFilter=XZY)
    Child=(Type=Control,Name="dyn_hose_upper14_end_ctrl")
    Parents=((Item=(Type=Bone,Name="thigh_twist_01_l"),Weight=0.8))
    Weight=1
   )
   > &RigUnit_ParentConstraint(
    AdvancedSettings=(InterpolationType=Average,RotationOrderForFilter=XZY)
    Child=(Type=Control,Name="dyn_hose_upper_end_ctrl")
    Parents=((Item=(Type=Bone,Name="spine_04"),Weight=1))
    Weight=1
   )
   > RigVMFunctionReturnNode()
   &RigUnit_ParentConstraint(
    AdvancedSettings=(InterpolationType=Average,RotationOrderForFilter=XZY)
    Child=(Type=Bone,Name="dyn_hose_ALT")
    ExecutePin=()
    Parents=((Item=(Type=Bone,Name="dyn_hose_end_base"),Weight=1),(Item=(Type=Bone,Name="dyn_hose_upper_base"),Weight=1))
    Weight=1
   )
  ]}
 New Function_ContainedGraph {nodes [
   RigVMFunctionEntryNode()
   > &RigUnit_RotationConstraintLocalSpaceOffset(
    AdvancedSettings=(InterpolationType=Average,RotationOrderForFilter=XZY)
    Child=(Type=Bone,Name="dyn_l_shoulderHelper")
    Filter=(bX=True,bY=True,bZ=True)
    Parents=((Item=(Type=Bone,Name="clavicle_l"),Weight=0),(Item=(Type=Bone,Name="spine_04"),Weight=1),(Item=(Type=Bone,Name="upperarm_twist_01_l"),Weight=0))
    Weight=0.5
   )
   > &RigUnit_RotationConstraintLocalSpaceOffset(
    AdvancedSettings=(InterpolationType=Average,RotationOrderForFilter=XZY)
    Child=(Type=Bone,Name="dyn_r_shoulderHelper")
    Filter=(bX=True,bY=True,bZ=True)
    Parents=((Item=(Type=Bone,Name="clavicle_r"),Weight=0),(Item=(Type=Bone,Name="spine_04"),Weight=1),(Item=(Type=Bone,Name="upperarm_twist_01_r"),Weight=0))
    Weight=0.5
   )
   > n1 = &RigUnit_SphericalPoseReader(
    ActiveRegionScaleFactors=(PositiveWidth=5,NegativeWidth=0.5,PositiveHeight=0.5,NegativeHeight=5)
    ActiveRegionSize=0.1
    Debug=(bDrawDebug=false,DebugScale=25,DebugSegments=20,DebugThickness=0.25)
    DriverAxis=(X=0,Y=1,Z=0)
    DriverItem=(Type=Bone,Name="spine_04")
    FalloffRegionScaleFactors=(PositiveWidth=1,NegativeWidth=1,PositiveHeight=1,NegativeHeight=1)
    FalloffSize=0.5
    OptionalParentItem=(Type=Bone)
    RotationOffset=(X=90,Y=0,Z=0)
   )
   > &RigUnit_ModifyTransforms(
    ItemToModify=((Item=(Type=Bone,Name="dyn_l_shoulderHelper"),Transform=(Rotation=(X=0,Y=-0.0872,Z=0,W=0.9962),Translation=(X=0,Y=0,Z=0),Scale3D=(X=1,Y=1,Z=1))))
    WeightMinimum=0
    Weight=@n1.OutputParam
   )
   > RigVMFunctionReturnNode()
   n2 = &RigUnit_SphericalPoseReader(
    ActiveRegionScaleFactors=(PositiveWidth=5,NegativeWidth=0.5,PositiveHeight=0.5,NegativeHeight=5)
    ActiveRegionSize=0.1
    Debug=(bDrawDebug=true,DebugScale=25,DebugSegments=20,DebugThickness=0.25)
    DriverAxis=(X=1,Y=0,Z=0)
    DriverItem=(Type=Bone,Name="upperarm_l")
    ExecutePin=()
    FalloffRegionScaleFactors=(PositiveWidth=1,NegativeWidth=0.1,PositiveHeight=0.1,NegativeHeight=1)
    FalloffSize=0.5
    OptionalParentItem=(Type=Bone)
    RotationOffset=(X=70,Y=-50,Z=0)
   )
   > &RigUnit_ModifyTransforms(
    ItemToModify=((Item=(Type=Bone,Name="dyn_l_shoulderHelper"),Transform=(Rotation=(X=0,Y=-0.2164,Z=0,W=0.9763),Translation=(X=0,Y=0,Z=0),Scale3D=(X=1,Y=1,Z=1))))
    WeightMinimum=0
    Weight=@n2.OutputParam
   )
   &RigUnit_ModifyTransforms(
    ExecutePin=()
    ItemToModify=((Item=(Type=Bone,Name="dyn_r_shoulderHelper"),Transform=(Rotation=(X=0,Y=0.2164,Z=0,W=0.9763),Translation=(X=0,Y=0,Z=0),Scale3D=(X=1,Y=1,Z=1))))
    Weight=1
    WeightMinimum=0
   )
   &RigUnit_SphericalPoseReader(
    ActiveRegionScaleFactors=(PositiveWidth=5,NegativeWidth=0.5,PositiveHeight=5,NegativeHeight=0.5)
    ActiveRegionSize=0.1
    Debug=(bDrawDebug=false,DebugScale=25,DebugSegments=20,DebugThickness=0.25)
    DriverAxis=(X=-1,Y=0,Z=0)
    DriverItem=(Type=Bone,Name="upperarm_r")
    ExecutePin=()
    FalloffRegionScaleFactors=(PositiveWidth=1,NegativeWidth=1,PositiveHeight=1,NegativeHeight=0.4)
    FalloffSize=0.5
    OptionalParentItem=(Type=Bone,Name="None")
    RotationOffset=(X=-180,Y=0,Z=-180)
   )
  ]}
 New Function_ContainedGraph {nodes [
   RigVMFunctionEntryNode()
   > &RigUnit_ParentConstraint(
    AdvancedSettings=(InterpolationType=Average,RotationOrderForFilter=XZY)
    Child=(Type=Bone,Name="dyn_fly")
    Parents=((Item=(Type=Bone,Name="thigh_l"),Weight=1),(Item=(Type=Bone,Name="thigh_r"),Weight=1))
    Weight=0.6
   )
   > n1 = &RigUnit_SphericalPoseReader(
    ActiveRegionScaleFactors=(PositiveWidth=3.5,NegativeWidth=1.5,PositiveHeight=3,NegativeHeight=3)
    ActiveRegionSize=0.1
    Debug=(bDrawDebug=false,bDraw2D=false,DebugScale=25,DebugSegments=20,DebugThickness=0.25)
    DriverAxis=(X=-1,Y=0,Z=0)
    DriverItem=(Type=Bone,Name="thigh_l")
    FalloffRegionScaleFactors=(PositiveWidth=1,NegativeWidth=1,PositiveHeight=1,NegativeHeight=1)
    FalloffSize=0.6
    OptionalParentItem=(Type=Bone,Name="pelvis")
    RotationOffset=(X=110,Y=0,Z=-20)
   )
   > &RigUnit_ModifyTransforms(
    ItemToModify=((Item=(Type=Bone,Name="dyn_fly"),Transform=(Rotation=(X=0,Y=0,Z=0.4226,W=0.9063),Translation=(X=-3,Y=0,Z=0),Scale3D=(X=0.2,Y=1,Z=1))))
    WeightMinimum=0.1
    Weight=@n1.OutputParam
   )
   > n2 = &RigUnit_SphericalPoseReader(
    ActiveRegionScaleFactors=(PositiveWidth=3.5,NegativeWidth=1.5,PositiveHeight=3,NegativeHeight=3)
    ActiveRegionSize=0.1
    Debug=(bDrawDebug=false,bDraw2D=false,DebugScale=25,DebugSegments=20,DebugThickness=0.25)
    DriverAxis=(X=1,Y=0,Z=0)
    DriverItem=(Type=Bone,Name="thigh_r")
    FalloffRegionScaleFactors=(PositiveWidth=1,NegativeWidth=1,PositiveHeight=1,NegativeHeight=1)
    FalloffSize=0.6
    OptionalParentItem=(Type=Bone,Name="pelvis")
    RotationOffset=(X=-110,Y=180,Z=-20)
   )
   > &RigUnit_ModifyTransforms(
    ItemToModify=((Item=(Type=Bone,Name="dyn_fly"),Transform=(Rotation=(X=0,Y=0,Z=0,W=1),Translation=(X=0,Y=0,Z=0),Scale3D=(X=1,Y=1,Z=1))))
    WeightMinimum=0.1
    Weight=@n2.OutputParam
   )
   > RigVMFunctionReturnNode()
  ]}
 New Function_ContainedGraph {nodes [
   RigVMFunctionEntryNode()
   > DeltaRotation = RigVMFunctionReferenceNode(
    ParentType=Null
    TwistAxis=(X=1,Y=0,Z=0)
    child=thigh_r
    debugRotation=False
    max=-90
    min=-5
    onlyTwist=False
    parent=thigh_r_null
    primAxis=1
    rotationOrder=YXZ
   )
   > &RigUnit_ModifyTransforms(
    ItemToModify=((Item=(Type=Bone,Name="dyn_gun_02"),Transform=(Rotation=(X=-0.2352,Y=0.2356,Z=0.6662,W=0.6674),Translation=(X=-0.8346,Y=6.0116,Z=-0.7385),Scale3D=(X=1,Y=1,Z=1))))
    WeightMinimum=0
    Weight=@DeltaRotation.ouputWeight
   )
   RigVMFunction_MathBoolToFloat = &RigVMFunction_MathBoolToFloat(Value=RigVMVariableNode() {variable=Sliding}.Value)
   AlphaInterp = &RigVMFunction_AlphaInterp(
    InterpSpeedDecreasing=10
    InterpSpeedIncreasing=10
    Scale=1
    bClampResult=False
    bInterpResult=False
    Value=@RigVMFunction_MathBoolToFloat.Result
   )
   Multiply = &RigVMFunction_MathDoubleMul(B=@AlphaInterp.Result, A=@DeltaRotation.ouputWeight)
   > &RigUnit_ModifyTransforms(
    ItemToModify=((Item=(Type=Bone,Name="dyn_gun_02"),Transform=(Rotation=(X=0,Y=0,Z=0,W=1),Translation=(X=5.1919,Y=-5.5225,Z=0.9878),Scale3D=(X=1,Y=1,Z=1))))
    WeightMinimum=0
    Weight=@Multiply.Result
   )
   > RigVMFunctionReturnNode()
   n1 = &RigUnit_SphericalPoseReader(
    ActiveRegionScaleFactors=(PositiveWidth=5.5,NegativeWidth=1,PositiveHeight=5,NegativeHeight=2)
    ActiveRegionSize=0.1
    Debug=(bDrawDebug=false,DebugScale=25,DebugSegments=20,DebugThickness=0.25)
    DriverAxis=(X=1,Y=0,Z=0)
    DriverItem=(Type=Bone,Name="thigh_r")
    ExecutePin=()
    FalloffRegionScaleFactors=(PositiveWidth=1,NegativeWidth=1,PositiveHeight=1,NegativeHeight=1)
    FalloffSize=0.45
    FlipHeightScaling=false
    FlipWidthScaling=true
    OptionalParentItem=(Type=Bone)
    RotationOffset=(X=0,Y=35,Z=-50)
   )
   > &RigUnit_ModifyTransforms(
    ItemToModify=((Item=(Type=Bone,Name="dyn_gun_02"),Transform=(Rotation=(X=-0.7424,Y=0.2424,Z=0.3462,W=0.5198),Translation=(X=0,Y=0,Z=-10),Scale3D=(X=1,Y=1,Z=1))))
    WeightMinimum=0
    Weight=@n1.OutputParam
   )
   > &RigUnit_ModifyTransforms(
    ItemToModify=((Item=(Type=Bone,Name="dyn_gun_02"),Transform=(Rotation=(X=0,Y=0,Z=0.0436,W=0.999),Translation=(X=0,Y=0,Z=0),Scale3D=(X=1,Y=1,Z=1))))
    Weight=1
    WeightMinimum=0
   )
   &RigUnit_ModifyTransforms(
    ExecutePin=()
    ItemToModify=((Item=(Type=Bone,Name="dyn_gun_02"),Transform=(Rotation=(X=-0.2021,Y=0.2025,Z=0.6769,W=0.6782),Translation=(X=-3.8358,Y=7.6049,Z=-1.7792),Scale3D=(X=1,Y=1,Z=1))))
    Weight=1
    WeightMinimum=0
   )
  ]}
 New Function_ContainedGraph {nodes [
   RigVMFunctionEntryNode()
   > n1 = &RigUnit_SphericalPoseReader(
    ActiveRegionScaleFactors=(PositiveWidth=3.5,NegativeWidth=1,PositiveHeight=3,NegativeHeight=3)
    ActiveRegionSize=0.1
    Debug=(bDrawDebug=false,bDraw2D=false,DebugScale=25,DebugSegments=20,DebugThickness=0.25)
    DriverAxis=(X=-1,Y=0,Z=0)
    DriverItem=(Type=Bone,Name="thigh_l")
    FalloffRegionScaleFactors=(PositiveWidth=1,NegativeWidth=1,PositiveHeight=1,NegativeHeight=1)
    FalloffSize=0.5
    OptionalParentItem=(Type=Bone,Name="pelvis")
    RotationOffset=(X=110,Y=0,Z=-20)
   )
   > &RigUnit_ModifyTransforms(
    ItemToModify=((Item=(Type=Bone,Name="dyn_belt_c_02"),Transform=(Rotation=(X=0,Y=-0.1736,Z=0,W=0.9848),Translation=(X=0,Y=0,Z=0),Scale3D=(X=1,Y=1,Z=1))))
    WeightMinimum=0.1
    Weight=@n1.OutputParam
   )
   > n2 = &RigUnit_SphericalPoseReader(
    ActiveRegionScaleFactors=(PositiveWidth=3.5,NegativeWidth=1.5,PositiveHeight=3,NegativeHeight=3)
    ActiveRegionSize=0.1
    Debug=(bDrawDebug=false,bDraw2D=false,DebugScale=25,DebugSegments=20,DebugThickness=0.25)
    DriverAxis=(X=1,Y=0,Z=0)
    DriverItem=(Type=Bone,Name="thigh_r")
    FalloffRegionScaleFactors=(PositiveWidth=1,NegativeWidth=1,PositiveHeight=1,NegativeHeight=1)
    FalloffSize=0.4
    OptionalParentItem=(Type=Bone,Name="pelvis")
    RotationOffset=(X=-110,Y=160,Z=0)
   )
   > &RigUnit_ModifyTransforms(
    ItemToModify=((Item=(Type=Bone,Name="dyn_suspenderBack_r"),Transform=(Rotation=(X=0,Y=0,Z=0,W=1),Translation=(X=0,Y=0,Z=10),Scale3D=(X=1,Y=1,Z=1))))
    WeightMinimum=0.1
    Weight=@n2.OutputParam
   )
   > &RigUnit_ModifyTransforms(
    ItemToModify=((Item=(Type=Bone,Name="dyn_belt_r_base"),Transform=(Rotation=(X=0,Y=0,Z=-0.0872,W=0.9962),Translation=(X=0,Y=0,Z=0),Scale3D=(X=1,Y=1,Z=1))))
    WeightMinimum=0
    Weight=@n2.OutputParam
   )
   > n3 = &RigUnit_SphericalPoseReader(
    ActiveRegionScaleFactors=(PositiveWidth=3.5,NegativeWidth=1,PositiveHeight=3,NegativeHeight=3)
    ActiveRegionSize=0.1
    Debug=(bDrawDebug=false,bDraw2D=false,DebugScale=25,DebugSegments=20,DebugThickness=0.25)
    DriverAxis=(X=-1,Y=0,Z=0)
    DriverItem=(Type=Bone,Name="thigh_l")
    FalloffRegionScaleFactors=(PositiveWidth=1,NegativeWidth=1,PositiveHeight=1,NegativeHeight=1)
    FalloffSize=0.4
    OptionalParentItem=(Type=Bone,Name="pelvis")
    RotationOffset=(X=-80,Y=0,Z=-30)
   )
   > &RigUnit_ModifyTransforms(
    ItemToModify=((Item=(Type=Bone,Name="dyn_belt_c_base"),Transform=(Rotation=(X=0,Y=0.342,Z=0,W=0.9397),Translation=(X=0,Y=0,Z=0),Scale3D=(X=1,Y=1,Z=1))))
    WeightMinimum=0.1
    Weight=@n3.OutputParam
   )
   > &RigUnit_ModifyTransforms(
    ItemToModify=((Item=(Type=Bone,Name="dyn_suspenderBack_l"),Transform=(Rotation=(X=0,Y=0,Z=0,W=1),Translation=(X=0,Y=-10,Z=0),Scale3D=(X=1,Y=1,Z=1))))
    WeightMinimum=0.1
    Weight=@n3.OutputParam
   )
   > n4 = &RigUnit_SphericalPoseReader(
    ActiveRegionScaleFactors=(PositiveWidth=3.5,NegativeWidth=1.5,PositiveHeight=3,NegativeHeight=3)
    ActiveRegionSize=0.1
    Debug=(bDrawDebug=false,bDraw2D=false,DebugScale=25,DebugSegments=20,DebugThickness=0.25)
    DriverAxis=(X=1,Y=0,Z=0)
    DriverItem=(Type=Bone,Name="thigh_r")
    FalloffRegionScaleFactors=(PositiveWidth=1,NegativeWidth=1,PositiveHeight=1,NegativeHeight=1)
    FalloffSize=0.4
    OptionalParentItem=(Type=Bone,Name="pelvis")
    RotationOffset=(X=100,Y=180,Z=-20)
   )
   > &RigUnit_ModifyTransforms(
    ItemToModify=((Item=(Type=Bone,Name="dyn_belt_c_base"),Transform=(Rotation=(X=0,Y=0.1736,Z=0,W=0.9848),Translation=(X=0,Y=0,Z=0),Scale3D=(X=1,Y=1,Z=1))))
    WeightMinimum=0.1
    Weight=@n4.OutputParam
   )
   > &RigUnit_ModifyTransforms(
    ItemToModify=((Item=(Type=Bone,Name="dyn_suspenderBack_r"),Transform=(Rotation=(X=0,Y=0,Z=0,W=1),Translation=(X=0,Y=-10,Z=0),Scale3D=(X=1,Y=1,Z=1))))
    WeightMinimum=0.1
    Weight=@n4.OutputParam
   )
   > RigVMFunctionReturnNode()
  ]}
 New Function_ContainedGraph {nodes [
   n1 = RigVMFunctionEntryNode()
   > n2 = RigVMDispatchNode(Array=(), Num=4) {template="DISPATCH_RigVMDispatch_ArraySetNum(io Array,in Num)"}
   n3 = RigVMVariableNode() {variable=Point 0}
   > n4 = RigVMDispatchNode(Index=0, Array=@n2.Array, Element=@n3.Value) {template="DISPATCH_RigVMDispatch_ArraySetAtIndex(io Array,in Index,in Element)"}
   n5 = RigVMVariableNode() {variable=Point 3}
   n6 = &RigVMFunction_MathVectorDistance(B=@n1.Point 0, A=@n1.Point 1)
   n7 = &RigVMFunction_MathFloatAdd(A=@n6.Result, B=&RigVMFunction_MathVectorDistance(A=@n1.Point 1, B=@n1.Point 2).Result) {template="Add::Execute(in A,in B,out Result)"}
   n8 = &RigVMFunction_MathFloatAdd(A=@n7.Result, B=&RigVMFunction_MathVectorDistance(A=@n1.Point 2, B=@n1.Point 3).Result) {template="Add::Execute(in A,in B,out Result)"}
   n9 = &RigVMFunction_MathFloatDiv(B=@n8.Result, A=@n6.Result) {template="Divide::Execute(in A,in B,out Result)"}
   b3t1 = RigVMCollapseNode(t=@n9.Result)
   n10 = RigVMVariableNode() {variable=Point 0}
   b0t1 = RigVMCollapseNode(t=@n9.Result)
   n11 = &RigVMFunction_MathVectorScale(Value=@n10.Value, Factor=@b0t1.Result) {template="Scale::Execute(in Value,in Factor,out Result)"}
   n12 = RigVMVariableNode() {variable=Point 1}
   n13 = &RigVMFunction_MathFloatDiv(A=@n7.Result, B=@n8.Result) {template="Divide::Execute(in A,in B,out Result)"}
   b3t2 = RigVMCollapseNode(t=@n13.Result)
   b0t2 = RigVMCollapseNode(t=@n13.Result)
   n14 = &RigVMFunction_MathVectorScale(Value=@n10.Value, Factor=@b0t2.Result) {template="Scale::Execute(in Value,in Factor,out Result)"}
   n15 = RigVMVariableNode() {variable=Point 2}
   b1t1 = RigVMCollapseNode(t=@n9.Result)
   b1t2 = RigVMCollapseNode(t=@n13.Result)
   b2t1 = RigVMCollapseNode(t=@n9.Result)
   b2t2 = RigVMCollapseNode(t=@n13.Result)
   Solve = RigVMCollapseNode(
    c=&RigVMFunction_MathVectorSub(
     B=&RigVMFunction_MathVectorScale(Value=@n5.Value, Factor=@b3t1.Result) {template="Scale::Execute(in Value,in Factor,out Result)"}.Result
     A=&RigVMFunction_MathVectorSub(B=@n11.Result, A=@n12.Value) {template="Subtract::Execute(in A,in B,out Result)"}.Result
    ) {template="Subtract::Execute(in A,in B,out Result)"}.Result
    f=&RigVMFunction_MathVectorSub(
     B=&RigVMFunction_MathVectorScale(Value=@n5.Value, Factor=@b3t2.Result) {template="Scale::Execute(in Value,in Factor,out Result)"}.Result
     A=&RigVMFunction_MathVectorSub(B=@n14.Result, A=@n15.Value) {template="Subtract::Execute(in A,in B,out Result)"}.Result
    ) {template="Subtract::Execute(in A,in B,out Result)"}.Result
    a=@b1t1.Result
    d=@b1t2.Result
    b=@b2t1.Result
    e=@b2t2.Result
   )
   > n16 = RigVMDispatchNode(Index=1, Array=@n4.Array, Element=@Solve.i) {template="DISPATCH_RigVMDispatch_ArraySetAtIndex(io Array,in Index,in Element)"}
   > n17 = RigVMDispatchNode(Index=2, Array=@n16.Array, Element=@Solve.j) {template="DISPATCH_RigVMDispatch_ArraySetAtIndex(io Array,in Index,in Element)"}
   n18 = RigVMVariableNode() {variable=Point 3}
   > n19 = RigVMDispatchNode(Index=3, Array=@n17.Array, Element=@n18.Value) {template="DISPATCH_RigVMDispatch_ArraySetAtIndex(io Array,in Index,in Element)"}
   n20 = /Script/ControlRigSpline.RigUnit_ControlRigSplineFromPoints(
    Points=((X=0,Y=0,Z=0),(X=0,Y=0,Z=0),(X=0,Y=0,Z=0),(X=0,Y=0,Z=0))
    SplineMode=BSpline
    bClosed=False
    SamplesPerSegment=@n1.SamplesPerSegment
    Points.1=@Solve.i
    Points.2=@Solve.j
    Points.0=@n3.Value
    Points.3=@n18.Value
   ) {is=struct RigVMUnitNode}
   > /Script/ControlRigSpline.RigUnit_DrawControlRigSpline(Color=(R=1,G=0,B=0,A=1), Detail=16, Thickness=2, Spline=@n20.Spline) {is=struct RigVMUnitNode}
   > RigVMFunctionReturnNode(Control Points=@n19.Array, Spline=@n20.Spline)
  ]}
 New Function_ContainedGraph {nodes [
   n1 = RigVMFunctionEntryNode()
   Clamp = &RigVMFunction_MathIntClamp(Value=@n1.primAxis)
   RigVMFunction_MathBoolToInteger = &RigVMFunction_MathBoolToInteger(Value=@n1.onlyTwist)
   RigVMFunction_MathQuaternionSwingTwist = &RigVMFunction_MathQuaternionSwingTwist(
    TwistAxis=@n1.TwistAxis
    Input=&RigUnit_GetRelativeTransformForItem(
     Child=(Type=Bone,Name="None")
     Parent=(Type=Bone,Name="None")
     bChildInitial=False
     bParentInitial=False
     Parent.Type=@n1.ParentType
     Child.Name=@n1.child
     Parent.Name=@n1.parent
    ).RelativeTransform.Rotation
   )
   RigVMFunction_MathQuaternionToEuler = &RigVMFunction_MathQuaternionToEuler(
    Value=RigVMDispatchNode(
     Values=((),())
     Index=@RigVMFunction_MathBoolToInteger.Result
     Values.0=@RigVMFunction_MathQuaternionSwingTwist.Swing
     Values.1=@RigVMFunction_MathQuaternionSwingTwist.Twist
    ) {template="DISPATCH_RigVMDispatch_SelectInt32(in Index,in Values,out Result)"}.Result
    RotationOrder=@n1.rotationOrder
   )
   DISPATCH_RigVMDispatch_SelectInt32 = RigVMDispatchNode(
    Values=(0,0,0)
    Index=@Clamp.Result
    Values.0=@RigVMFunction_MathQuaternionToEuler.Result.X
    Values.1=@RigVMFunction_MathQuaternionToEuler.Result.Y
    Values.2=@RigVMFunction_MathQuaternionToEuler.Result.Z
   ) {template="DISPATCH_RigVMDispatch_SelectInt32(in Index,in Values,out Result)"}
   Remap = &RigVMFunction_MathDoubleRemap(
    TargetMaximum=1
    TargetMinimum=0
    bClamp=false
    Value=@DISPATCH_RigVMDispatch_SelectInt32.Result
    SourceMaximum=@n1.max
    SourceMinimum=@n1.min
   )
   > RigVMFunctionReturnNode(ouputWeight=&RigVMFunction_MathDoubleClamp(Value=@Remap.Result).Result)
  ]}
 RigVMFunctionLibrary {nodes [
   Collar solution = RigVMCollapseNode(ExecuteContext=())
   DeltaRotation = RigVMCollapseNode(ExecuteContext=(), TwistAxis=(X=0,Y=0,Z=0), debugRotation=False, onlyTwist=False)
   "Spline four points " = RigVMCollapseNode(ExecuteContext=(), Point 0=(X=0,Y=0,Z=0), Point 1=(X=0,Y=0,Z=0), Point 2=(X=0,Y=0,Z=0), Point 3=(X=0,Y=0,Z=0), SamplesPerSegment=16)
   "fly solution " = RigVMCollapseNode(ExecuteContext=())
   gun solution = RigVMCollapseNode(ExecuteContext=())
   "hose and belt solution " = RigVMCollapseNode(ExecuteContext=())
   "jacket solution " = RigVMCollapseNode(ExecuteContext=())
   suspender crouch solution = RigVMCollapseNode(ExecuteContext=())
  ]}
 RigVMFunction_Sequence_1_ContainedGraph {nodes [
   n1 = RigVMFunctionEntryNode()
   RigVMFunction_Sequence = &RigVMFunction_Sequence(ExecuteContext=@n1.ExecuteContext)
   n2 = &RigVMFunction_Sequence(ExecuteContext=@RigVMFunction_Sequence.B)
   RigVMFunctionReturnNode(A=@RigVMFunction_Sequence.A, B=@n2.A, C=@n2.B)
  ]}
 RigVMModel {nodes [
   RigUnit_BeginExecution = &RigUnit_BeginExecution()
   FN_DeformRig_FNC_2 = RigVMFunctionReferenceNode(
    DeformRig_ Curve="(deformRig_l_upperarm_up_alpha_72_#1=False,deformRig_r_upperarm_up_alpha_73_#2=False,deformRig_l_upperarm_fr_alpha_74_#3=False,deformRig_r_upperarm_fr_alpha_75_CB51B8 ...(truncated)"
    DeformRig_Bool="(upperarm_1_#4=True,elbow_3_#5=True,wrist_7_#6=True,groin_8_#7=True,glute_17_#8=True,knee_9_40C408324CA261 ...(truncated)"
    DeformRig_Debug="(upperarm_debug=False,wrist_debug=False,lat_debug=False,groinGlute_debug=False,upperarm_debug_18_#4=False,wrist_debug_20_#6=False,groinGlute_debug_27_#7=False,lat_deb ...(truncated)"
    DeformRig_Value="(upperarm_up_18_#9=(Translation=(X=0,Y=0,Z=8),Scale3D=(X=0.7,Y=1,Z=1)),upperarm_fwd_fr_19_#10=(Rotation=(X=0,Y=0,Z=0,W=1),Translation=(X=0,Y=0,Z=0),Scale3D=(X=0.5,Y=1,Z=1)),upperarm_fwd_bk_20_9973A ...(truncated)"
    ExecuteContext=@RigUnit_BeginExecution.ExecutePin
   )
   > Collar solution = RigVMFunctionReferenceNode()
   > "hose and belt solution " = RigVMFunctionReferenceNode()
   > "jacket solution " = RigVMFunctionReferenceNode()
   > "fly solution " = RigVMFunctionReferenceNode()
   > gun solution = RigVMFunctionReferenceNode()
   > suspender crouch solution = RigVMFunctionReferenceNode()
   FN_DeformRig_FNC = RigVMFunctionReferenceNode(
    DeformRig_ Curve="(deformRig_l_upperarm_up_alpha_72_#1=False,deformRig_r_upperarm_up_alpha_73_#2=False,deformRig_l_upperarm_fr_alpha_74_#3=False,deformRig_r_upperarm_fr_alpha_75_CB51B8 ...(truncated)"
    DeformRig_Bool="(upperarm_1_#4=True,elbow_3_#5=True,wrist_7_#6=True,groin_8_#7=True,glute_17_#8=True,knee_9_40C408324CA261 ...(truncated)"
    DeformRig_Debug="(upperarm_debug=False,wrist_debug=False,lat_debug=False,groinGlute_debug=False,upperarm_debug_18_#4=False,wrist_debug_20_#6=False,groinGlute_debug_27_#7=False,lat_deb ...(truncated)"
    DeformRig_Value="(upperarm_up_18_#9=(Translation=(X=0,Y=0,Z=8),Scale3D=(X=0.7,Y=1,Z=1)),upperarm_fwd_fr_19_#10=(Translation=(X=0,Y=10,Z=0),Scale3D=(X=0.5,Y=1,Z=1)),upperarm_fwd_bk_20_9973AAD84EE0A1E349BC9E85067D4A3 ...(truncated)"
    ExecuteContext=()
   )
  ]}
]
hierarchy [
 root {loc=<0 0 0>, rot=<0 0 0 1>}
 pelvis {parent=root, loc=<0 -0.7071 93.1191>, rot=<0.0127 0.707 -0.0127 -0.707>}
 spine_01 {parent=pelvis, loc=<6.7494 -0.6388 0>, rot=<0 0 0.034 -0.9994>}
 spine_02 {parent=spine_01, loc=<7.2518 -0.035 0>, rot=<0 0 0.0238 -0.9997>}
 spine_03 {parent=spine_02, loc=<6.0104 0 0>, rot=<0 0 0.0149 -0.9999>}
 spine_04 {parent=spine_03, loc=<8.1024 0 0>, rot=<0 0 -0.0132 -0.9999>}
 spine_05 {parent=spine_04, loc=<10.0146 0 0>, rot=<0 0 -0.0015 -1>}
 neck_01 {parent=spine_05, loc=<11.3138 0.1529 0>, rot=<0 0 -0.2003 -0.9797>}
 neck_02 {parent=neck_01, loc=<4.5627 0 0>, rot=<0 0 0.0504 -0.9987>}
 head {parent=neck_02, loc=<5.3914 0 0>, rot=<0 0 0.115 -0.9934>}
 attach_cape {parent=spine_05, loc=<9.0205 -11.7158 0.0001>, rot=<0.0616 -0.7044 -0.0616 -0.7044>}
 cape {parent=attach_cape, type=Null, loc=<0 0 0>, rot=<0 0 0 1>}
 attach_backpack {parent=spine_05, loc=<-0.1165 -13.4755 0.0001>, rot=<-0.0092 -0.707 0.0092 -0.707>}
 Backpack_BR {parent=attach_backpack, type=Null, loc=<0 -0.0196 0>, rot=<0 0 0 1>}
 clavicle_l {parent=spine_05, loc=<4.3575 -0.1466 -2.3223>, rot=<0.0465 -0.7289 -0.009 -0.6829>}
 upperarm_l {parent=clavicle_l, loc=<15.5111 0 0>, rot=<0.0226 -0.3234 -0.0088 -0.9459>}
 lowerarm_l {parent=upperarm_l, loc=<26.6514 0 0>, rot=<0.0391 -0.0099 -0.2453 -0.9686>}
 hand_l {parent=lowerarm_l, loc=<22.3491 0 0>, rot=<-0.7081 -0.0507 0.0975 -0.6975>}
 thumb_01_l {parent=hand_l, loc=<1.3118 -0.3191 -2.459>, rot=<0.6546 -0.4042 0.0664 -0.6354>}
 thumb_02_l {parent=thumb_01_l, loc=<5.9336 -0.0448 0.0401>, rot=<-0.0037 0.0213 0.298 -0.9543>}
 thumb_03_l {parent=thumb_02_l, loc=<2.2223 0.3682 -0.0907>, rot=<0.0118 -0.013 0.0066 -0.9998>}
 index_metacarpal_l {parent=hand_l, loc=<1.6535 0.0183 -2.96>, rot=<0.0158 -0.0124 -0.0173 -0.9996>}
 index_01_l {parent=index_metacarpal_l, loc=<8.3008 0.0357 -0.0524>, rot=<-0.0044 0.0485 0.0924 -0.9945>}
 index_02_l {parent=index_01_l, loc=<3.3587 -0.3004 0.0452>, rot=<-0.002 0.0027 0.3225 -0.9466>}
 index_03_l {parent=index_02_l, loc=<2.5578 -0.123 -0.0252>, rot=<-0.002 0.0067 0.2558 -0.9667>}
 middle_metacarpal_l {parent=hand_l, loc=<2.1856 -0.1725 -0.5984>, rot=<0.006 0.0168 -0.0329 -0.9993>}
 middle_01_l {parent=middle_metacarpal_l, loc=<7.3275 -0.2447 0.0129>, rot=<-0.0208 0.0042 0.0858 -0.9961>}
 middle_02_l {parent=middle_01_l, loc=<4.0123 0.1401 0.1175>, rot=<-0.0218 0.0258 0.3512 -0.9357>}
 middle_03_l {parent=middle_02_l, loc=<2.6077 0.0534 0.1373>, rot=<0.0082 0.0188 0.1708 -0.9851>}
 ring_metacarpal_l {parent=hand_l, loc=<2.2074 -0.3625 1.3638>, rot=<-0.0284 0.0722 -0.0458 -0.9959>}
 ring_01_l {parent=ring_metacarpal_l, loc=<7.0295 -0.6197 -0.1446>, rot=<-0.0489 -0.0252 0.1892 -0.9804>}
 ring_02_l {parent=ring_01_l, loc=<3.6174 0.4647 -0.1027>, rot=<0 0 0.2806 -0.9598>}
 ring_03_l {parent=ring_02_l, loc=<2.4152 0.114 0.0199>, rot=<0 0 0.1189 -0.9929>}
 pinky_metacarpal_l {parent=hand_l, loc=<1.7616 -0.9753 3.3814>, rot=<-0.0397 0.0814 -0.069 -0.9935>}
 pinky_01_l {parent=pinky_metacarpal_l, loc=<7.1709 -0.4831 -0.1094>, rot=<-0.0654 -0.0035 0.2242 -0.9723>}
 pinky_02_l {parent=pinky_01_l, loc=<2.8274 0.0893 -0.1043>, rot=<0.0037 0.0011 0.2854 -0.9584>}
 pinky_03_l {parent=pinky_02_l, loc=<2.2046 0.0307 -0.0078>, rot=<-0.0002 0.0098 0.1445 -0.9895>}
 lowerarm_twist_02_l {parent=lowerarm_l, loc=<9.6199 0 0>, rot=<0 0 0 -1>}
 lowerarm_twist_01_l {parent=lowerarm_l, loc=<17.8549 0 0>, rot=<0 0 0 -1>}
 upperarm_twist_01_l {parent=upperarm_l, loc=<0.0041 0 0>, rot=<0 0 0 -1>}
 deform_upperarm_bk_base_l {parent=upperarm_twist_01_l, loc=<1.0813 -7.2193 -1.9453>, rot=<0 0 0 -1>}
 deform_upperarm_bk_l {parent=deform_upperarm_bk_base_l, loc=<0 0 0>, rot=<0 0 0 1>}
 deform_upperarm_fr_base_l {parent=upperarm_twist_01_l, loc=<-0.3691 7.0506 -2.765>, rot=<0 0 0 -1>}
 deform_upperarm_fr_l {parent=deform_upperarm_fr_base_l, loc=<0 0 0>, rot=<0 0 0 1>}
 upperarm_twist_02_l {parent=upperarm_l, loc=<13.7941 0 0>, rot=<0 0 0 -1>}
 dyn_l_shoulderHelper {parent=clavicle_l, loc=<13.6424 10.1872 -1.0711>, rot=<0.0008 0.0336 -0.0252 -0.9991>}
 clavicle_r {parent=spine_05, loc=<4.3575 -0.1466 2.3223>, rot=<0.7289 0.0465 -0.6829 0.009>}
 upperarm_r {parent=clavicle_r, loc=<-15.5111 0 0>, rot=<0.0226 -0.3234 -0.0088 -0.9459>}
 lowerarm_r {parent=upperarm_r, loc=<-26.6514 0 0>, rot=<0.0391 -0.0099 -0.2453 -0.9686>}
 hand_r {parent=lowerarm_r, loc=<-22.3491 0 0>, rot=<-0.7081 -0.0507 0.0975 -0.6975>}
 thumb_01_r {parent=hand_r, loc=<-1.3118 0.3191 2.459>, rot=<0.6546 -0.4042 0.0664 -0.6354>}
 thumb_02_r {parent=thumb_01_r, loc=<-5.9336 0.0448 -0.0401>, rot=<-0.0037 0.0213 0.298 -0.9543>}
 thumb_03_r {parent=thumb_02_r, loc=<-2.2223 -0.3682 0.0907>, rot=<0.0118 -0.013 0.0066 -0.9998>}
 index_metacarpal_r {parent=hand_r, loc=<-1.6535 -0.0183 2.96>, rot=<0.0158 -0.0124 -0.0173 -0.9996>}
 index_01_r {parent=index_metacarpal_r, loc=<-8.3008 -0.0357 0.0524>, rot=<-0.0044 0.0485 0.0924 -0.9945>}
 index_02_r {parent=index_01_r, loc=<-3.3587 0.3004 -0.0452>, rot=<-0.002 0.0027 0.3225 -0.9466>}
 index_03_r {parent=index_02_r, loc=<-2.5578 0.123 0.0252>, rot=<-0.002 0.0067 0.2558 -0.9667>}
 middle_metacarpal_r {parent=hand_r, loc=<-2.1856 0.1725 0.5984>, rot=<0.006 0.0168 -0.0329 -0.9993>}
 middle_01_r {parent=middle_metacarpal_r, loc=<-7.3275 0.2447 -0.0129>, rot=<-0.0208 0.0042 0.0858 -0.9961>}
 middle_02_r {parent=middle_01_r, loc=<-4.0123 -0.1401 -0.1175>, rot=<-0.0218 0.0258 0.3512 -0.9357>}
 middle_03_r {parent=middle_02_r, loc=<-2.6077 -0.0534 -0.1373>, rot=<0.0082 0.0188 0.1708 -0.9851>}
 ring_metacarpal_r {parent=hand_r, loc=<-2.2074 0.3625 -1.3638>, rot=<-0.0284 0.0722 -0.0458 -0.9959>}
 ring_01_r {parent=ring_metacarpal_r, loc=<-7.0295 0.6197 0.1446>, rot=<-0.0489 -0.0252 0.1892 -0.9804>}
 ring_02_r {parent=ring_01_r, loc=<-3.6174 -0.4647 0.1027>, rot=<0 0 0.2806 -0.9598>}
 ring_03_r {parent=ring_02_r, loc=<-2.4152 -0.114 -0.0199>, rot=<0 0 0.1189 -0.9929>}
 pinky_metacarpal_r {parent=hand_r, loc=<-1.7616 0.9753 -3.3814>, rot=<-0.0397 0.0814 -0.069 -0.9935>}
 pinky_01_r {parent=pinky_metacarpal_r, loc=<-7.1709 0.4831 0.1094>, rot=<-0.0654 -0.0035 0.2242 -0.9723>}
 pinky_02_r {parent=pinky_01_r, loc=<-2.8274 -0.0893 0.1043>, rot=<0.0037 0.0011 0.2854 -0.9584>}
 pinky_03_r {parent=pinky_02_r, loc=<-2.2046 -0.0307 0.0078>, rot=<-0.0002 0.0098 0.1445 -0.9895>}
 lowerarm_twist_02_r {parent=lowerarm_r, loc=<-9.6199 0 0>, rot=<0 0 0 -1>}
 lowerarm_twist_01_r {parent=lowerarm_r, loc=<-17.8549 0 0>, rot=<0 0 0 -1>}
 upperarm_twist_01_r {parent=upperarm_r, loc=<-0.0041 0 0>, rot=<0 0 0 -1>}
 deform_upperarm_bk_base_r {parent=upperarm_twist_01_r, loc=<-1.0814 7.2192 1.9455>, rot=<0 0 0 -1>}
 deform_upperarm_bk_r {parent=deform_upperarm_bk_base_r, loc=<0 0 0>, rot=<0 0 0 1>}
 deform_upperarm_fr_base_r {parent=upperarm_twist_01_r, loc=<0.3689 -7.0506 2.7652>, rot=<0 0 0 -1>}
 deform_upperarm_fr_r {parent=deform_upperarm_fr_base_r, loc=<0 0 0>, rot=<0 0 0 1>}
 upperarm_twist_02_r {parent=upperarm_r, loc=<-13.7941 0 0>, rot=<0 0 0 -1>}
 dyn_r_shoulderHelper {parent=clavicle_r, loc=<-13.6425 -10.1872 1.0709>, rot=<0.0008 0.0336 -0.0252 -0.9991>}
 dyn_l_collar {parent=spine_05, loc=<8.0658 9.6118 -4.6522>, rot=<-0.049 -0.3527 -0.1925 -0.9144>}
 dyn_r_collar {parent=spine_05, loc=<8.0657 9.6118 4.6522>, rot=<-0.3471 0.0797 0.894 -0.2719>}
 dyn_neckNoTwist {parent=spine_05, loc=<11.3138 0.1529 0>, rot=<0 0 -0.2003 -0.9797>}
 dyn_c_collar {parent=dyn_neckNoTwist, loc=<-0.6131 -6.3373 0.01>, rot=<0 0 0.1608 -0.987>}
 dyn_detail_main {parent=spine_04, loc=<7.0736 12.5455 0.1678>, rot=<0.0416 -0.9991 0 0>}
 dyn_detail_base {parent=dyn_detail_main, loc=<-0.2652 3.8857 0.0918>, rot=<0 0 0 -1>}
 dyn_hose_upper_base {parent=dyn_detail_base, loc=<6.7388 1.028 -2.3823>, rot=<0 0 0 -1>}
 dyn_hose_upper {parent=dyn_hose_upper_base, loc=<0 0 0>, rot=<0 0 0 1>}
 dyn_hose_upper1 {parent=dyn_hose_upper, loc=<2.608 0.7192 0.1542>, rot=<0.0098 0.0621 -0.0825 -0.9946>}
 dyn_hose_upper2 {parent=dyn_hose_upper1, loc=<4.9253 0.428 -0.0069>, rot=<0 0.0479 0 -0.9989>}
 dyn_hose_upper3 {parent=dyn_hose_upper2, loc=<4.8029 0.1574 0>, rot=<0 0 0 1>}
 dyn_hose_upper4 {parent=dyn_hose_upper3, loc=<4.4961 -0.1147 0>, rot=<0 0 0.0178 -0.9998>}
 dyn_hose_upper5 {parent=dyn_hose_upper4, loc=<4.7659 0.0558 -0.1086>, rot=<0.0002 0.0436 0.0199 -0.9989>}
 dyn_hose_upper6 {parent=dyn_hose_upper5, loc=<4.9342 -0.0957 0>, rot=<0 0 0.06 -0.9982>}
 dyn_hose_upper7 {parent=dyn_hose_upper6, loc=<4.6223 0.1827 0.4831>, rot=<-0.0002 0.0197 0.0295 -0.9994>}
 dyn_hose_upper8 {parent=dyn_hose_upper7, loc=<5.1536 0.3463 0.1694>, rot=<-0.0024 0.0747 0.0318 -0.9967>}
 dyn_hose_upper9 {parent=dyn_hose_upper8, loc=<4.4975 0.2314 -0.4675>, rot=<0.0084 0.0426 0.0932 -0.9947>}
 dyn_hose_upper10 {parent=dyn_hose_upper9, loc=<4.7755 0.156 -0.0019>, rot=<-0.0003 0.0009 0.0655 -0.9978>}
 dyn_hose_upper11 {parent=dyn_hose_upper10, loc=<4.5553 0.0775 0.2834>, rot=<0.0022 0.0312 0.0909 -0.9954>}
 dyn_hose_upper12 {parent=dyn_hose_upper11, loc=<4.5873 -0.0487 0.0571>, rot=<0.0016 0.0218 0.0901 -0.9957>}
 dyn_hose_upper13 {parent=dyn_hose_upper12, loc=<4.3403 0.3938 0.5919>, rot=<0.0003 0.007 0.0426 -0.9991>}
 dyn_hose_upper14 {parent=dyn_hose_upper13, loc=<4.3488 0.2495 0.061>, rot=<0 -0.0794 0 -0.9968>}
 dyn_pocket {parent=spine_03, loc=<-2.2305 8.7377 13.3113>, rot=<0.0387 -0.706 -0.0387 -0.706>}
 dyn_hose_ALT {parent=spine_01, loc=<-0.5154 19.4766 -2.1077>, rot=<0.0114 -0.707 -0.0114 -0.707>}
 thigh_l {parent=pelvis, loc=<-2.5693 -0.0109 -7.7501>, rot=<0.0012 0.0572 0.0155 -0.9982>}
 calf_l {parent=thigh_l, loc=<-39.0484 0 0>, rot=<0.0006 -0.0161 -0.0391 -0.9991>}
 foot_l {parent=calf_l, loc=<-42.561 0.1077 0.4292>, rot=<0.0418 -0.0455 0.0343 -0.9975>}
 ball_l {parent=foot_l, loc=<-6.1599 13.8264 0.6806>, rot=<0.0001 -0.0001 -0.7063 -0.7079>}
 calf_twist_02_l {parent=calf_l, loc=<-18.8064 -0.0967 0.0397>, rot=<0.0237 0.0095 -0.0035 -0.9997>}
 calf_twist_01_l {parent=calf_l, loc=<-35.813 -0.0967 0.0397>, rot=<0.0237 0.0095 -0.0035 -0.9997>}
 deform_knee_out_base_l {parent=calf_l, loc=<0 4 0>, rot=<0 0 0 1>}
 deform_knee_out_l {parent=deform_knee_out_base_l, loc=<0 0 0>, rot=<0 0 0 1>}
 thigh_twist_01_l {parent=thigh_l, loc=<-21.6303 0 0>, rot=<0.0013 0 0 -1>}
 dyn_hose_end_base {parent=thigh_twist_01_l, loc=<0 0 0>, rot=<-0.0023 -0.9984 -0.0015 0.0572>}
 dyn_hose_end {parent=dyn_hose_end_base, loc=<0 0 0>, rot=<0 0 0 1>}
 dyn_belt_l_base {parent=thigh_l, loc=<5.8191 0.7829 -7.7766>, rot=<0.0417 -0.7058 -0.7057 -0.0451>}
 dyn_belt_l_01 {parent=dyn_belt_l_base, loc=<0 0 0>, rot=<0 0 0 1>}
 dyn_belt_l_02 {parent=dyn_belt_l_01, loc=<24.5295 0 0>, rot=<0 0 0 1>}
 dyn_suspenderBack_l {parent=dyn_belt_l_01, loc=<22.2225 6.5416 -11.5311>, rot=<-0.5557 0.4373 0.4373 -0.5557>}
 deform_groin_base_l {parent=thigh_l, loc=<-2.1908 8.1142 -1.7442>, rot=<0 0 0 -1>}
 deform_groin_l {parent=deform_groin_base_l, loc=<0 0 0>, rot=<0 0 0 1>}
 thigh_r {parent=pelvis, loc=<-2.5693 -0.0109 7.7501>, rot=<0.0572 -0.0012 0.9982 0.0155>}
 calf_r {parent=thigh_r, loc=<39.0484 0 0>, rot=<0.0006 -0.0161 -0.0391 -0.9991>}
 foot_r {parent=calf_r, loc=<42.561 -0.1077 -0.4292>, rot=<0.0418 -0.0455 0.0343 -0.9975>}
 ball_r {parent=foot_r, loc=<6.1599 -13.8264 -0.6806>, rot=<0.0001 -0.0001 -0.7063 -0.7079>}
 calf_twist_02_r {parent=calf_r, loc=<18.8064 0.0967 -0.0397>, rot=<0.0237 0.0095 -0.0035 -0.9997>}
 calf_twist_01_r {parent=calf_r, loc=<35.813 0.0967 -0.0397>, rot=<0.0237 0.0095 -0.0035 -0.9997>}
 deform_knee_out_base_r {parent=calf_r, loc=<0 -4 0>, rot=<0 0 0 1>}
 deform_knee_out_r {parent=deform_knee_out_base_r, loc=<0 0 0>, rot=<0 0 0 1>}
 thigh_twist_01_r {parent=thigh_r, loc=<21.6303 0 0>, rot=<0.0013 0 0 -1>}
 dyn_belt_r_base {parent=thigh_r, loc=<-5.8191 -0.7829 7.7766>, rot=<0.0417 -0.7058 -0.7057 -0.0451>}
 dyn_belt_r_01 {parent=dyn_belt_r_base, loc=<0 0 0>, rot=<0 0 0 1>}
 dyn_belt_r_02 {parent=dyn_belt_r_01, loc=<-24.5296 0 0>, rot=<0 0 0 1>}
 dyn_suspenderBack_r {parent=dyn_belt_r_01, loc=<-22.2226 -6.5416 11.5311>, rot=<0.5557 -0.4373 0.4373 -0.5557>}
 deform_groin_base_r {parent=thigh_r, loc=<2.1908 -8.1142 1.7442>, rot=<0 0 0 -1>}
 deform_groin_r {parent=deform_groin_base_r, loc=<0 0 0>, rot=<0 0 0 1>}
 dyn_gun_01 {parent=pelvis, loc=<8.2733 -0.0099 14.0872>, rot=<0.1067 0.6873 0.705 -0.1387>}
 dyn_gun_02 {parent=dyn_gun_01, loc=<6.5835 0 0>, rot=<0 0 0 1>}
 dyn_belt_c_base {parent=pelvis, loc=<5.1933 -12.4244 0>, rot=<-0.1284 0.6953 0.6953 -0.1284>}
 dyn_belt_c_01 {parent=dyn_belt_c_base, loc=<0 0 0>, rot=<0 0 0 1>}
 dyn_belt_c_02 {parent=dyn_belt_c_01, loc=<4.7916 0 0>, rot=<0 0 0 1>}
 dyn_belt_c_03 {parent=dyn_belt_c_02, loc=<4.7916 0 0>, rot=<0 0 0 1>}
 dyn_hose_upper14_end {parent=pelvis, loc=<-22.7261 7.2908 -20.363>, rot=<0.4084 -0.8688 0.1324 -0.2467>}
 dyn_hose_upper14_end_ctrl {controlType=EULER_TRANSFORM, parent=dyn_hose_upper14_end, type=Control, loc=<0 0 0>, rot=<0 0 0 1>}
 dyn_hose_upper13_end {parent=pelvis, loc=<-20.6337 10.4747 -18.2507>, rot=<0.4176 -0.8465 0.0996 -0.3149>}
 dyn_hose_upper13_end_ctrl {controlType=EULER_TRANSFORM, parent=dyn_hose_upper13_end, type=Control, loc=<0 0 0>, rot=<0 0 0 1>}
 dyn_hose_upper12_end {parent=pelvis, loc=<-18.5242 13.2526 -15.5716>, rot=<0.3804 -0.8656 0.0893 -0.3131>}
 dyn_hose_upper12_end_ctrl {controlType=EULER_TRANSFORM, parent=dyn_hose_upper12_end, type=Control, loc=<0 0 0>, rot=<0 0 0 1>}
 dyn_hose_upper11_end {parent=pelvis, loc=<-15.6456 15.9549 -13.2351>, rot=<0.2984 -0.9028 0.0703 -0.3016>}
 dyn_hose_upper11_end_ctrl {controlType=EULER_TRANSFORM, parent=dyn_hose_upper11_end, type=Control, loc=<0 0 0>, rot=<0 0 0 1>}
 dyn_hose_upper10_end {parent=pelvis, loc=<-12.3354 17.8232 -10.7075>, rot=<0.2121 -0.935 0.0539 -0.2791>}
 dyn_hose_upper10_end_ctrl {controlType=EULER_TRANSFORM, parent=dyn_hose_upper10_end, type=Control, loc=<0 0 0>, rot=<0 0 0 1>}
 dyn_hose_upper9_end {parent=pelvis, loc=<-8.4883 19.13 -8.1932>, rot=<0.1504 -0.9472 0.0354 -0.2811>}
 dyn_hose_upper9_end_ctrl {controlType=EULER_TRANSFORM, parent=dyn_hose_upper9_end, type=Control, loc=<0 0 0>, rot=<0 0 0 1>}
 dyn_hose_upper8_end {parent=pelvis, loc=<-4.3104 19.4437 -6.4766>, rot=<0.0574 -0.9678 0.0235 -0.2439>}
 dyn_hose_upper8_end_ctrl {controlType=EULER_TRANSFORM, parent=dyn_hose_upper8_end, type=Control, loc=<0 0 0>, rot=<0 0 0 1>}
 dyn_hose_upper7_end {parent=pelvis, loc=<0.4915 19.3907 -4.567>, rot=<0.0253 -0.9847 0.0176 -0.1714>}
 dyn_hose_upper7_end_ctrl {controlType=EULER_TRANSFORM, parent=dyn_hose_upper7_end, type=Control, loc=<0 0 0>, rot=<0 0 0 1>}
 dyn_hose_upper6_end {parent=pelvis, loc=<4.751 19.2021 -2.7087>, rot=<-0.004 -0.9882 0.0128 -0.1525>}
 dyn_hose_upper6_end_ctrl {controlType=EULER_TRANSFORM, parent=dyn_hose_upper6_end, type=Control, loc=<0 0 0>, rot=<0 0 0 1>}
 dyn_hose_upper5_end {parent=pelvis, loc=<9.4269 18.6864 -1.2166>, rot=<-0.0633 -0.9862 0.0037 -0.153>}
 dyn_hose_upper5_end_ctrl {controlType=EULER_TRANSFORM, parent=dyn_hose_upper5_end, type=Control, loc=<0 0 0>, rot=<0 0 0 1>}
 dyn_hose_upper4_end {parent=pelvis, loc=<14.0265 17.8437 -0.2881>, rot=<-0.0831 -0.9905 -0.0019 -0.1099>}
 dyn_hose_upper4_end_ctrl {controlType=EULER_TRANSFORM, parent=dyn_hose_upper4_end, type=Control, loc=<0 0 0>, rot=<0 0 0 1>}
 dyn_hose_upper3_end {parent=pelvis, loc=<18.3458 17.0574 0.6883>, rot=<-0.1006 -0.9888 -0.0039 -0.1098>}
 dyn_hose_upper3_end_ctrl {controlType=EULER_TRANSFORM, parent=dyn_hose_upper3_end, type=Control, loc=<0 0 0>, rot=<0 0 0 1>}
 dyn_hose_upper2_end {parent=pelvis, loc=<22.9044 15.9432 1.723>, rot=<-0.1006 -0.9888 -0.0039 -0.1098>}
 dyn_hose_upper2_end_ctrl {controlType=EULER_TRANSFORM, parent=dyn_hose_upper2_end, type=Control, loc=<0 0 0>, rot=<0 0 0 1>}
 dyn_hose_upper1_end {parent=pelvis, loc=<27.6084 14.5371 2.3041>, rot=<-0.1003 -0.993 -0.0087 -0.0623>}
 dyn_hose_upper1_end_ctrl {controlType=EULER_TRANSFORM, parent=dyn_hose_upper1_end, type=Control, loc=<0 0 0>, rot=<0 0 0 1>}
 dyn_hose_upper_end {parent=pelvis, loc=<30.1889 13.7249 2.4583>, rot=<-0.0179 -0.9998 0 0>}
 dyn_hose_upper_end_ctrl {controlType=EULER_TRANSFORM, parent=dyn_hose_upper_end, type=Control, loc=<0 0 0>, rot=<0 0 0 1>}
 dyn_belt_helper {parent=pelvis, loc=<0 0 0>, rot=<0 0 0 1>}
 dyn_strapHelper {parent=dyn_belt_helper, loc=<-4.2941 8.1593 -9.7386>, rot=<0.0012 0.0572 0.0155 -0.9982>}
 dyn_strapHelper_centered {parent=dyn_belt_helper, loc=<-0.8151 10.4641 -11.422>, rot=<0.2033 0.073 0.0809 -0.973>}
 dyn_strapHelper_l {parent=dyn_belt_helper, loc=<-0.8151 10.4641 -11.422>, rot=<0.2033 0.073 0.0809 -0.973>}
 dyn_strapHelper_r {parent=dyn_belt_helper, loc=<-1.2944 12.1013 7.7992>, rot=<0.0963 -0.1508 0.9811 0.0735>}
 dyn_fly {parent=pelvis, loc=<-7.4993 11.7261 -0.7766>, rot=<-0.0179 -0.9998 0 0>}
 deform_glute_base_l {parent=pelvis, loc=<-8.8064 -8.8222 -8.4287>, rot=<0.0275 0.0501 -0.4472 -0.8926>}
 deform_glute_l {parent=deform_glute_base_l, loc=<0 0 0>, rot=<0 0 0 1>}
 deform_glute_base_r {parent=pelvis, loc=<-8.8062 -8.8222 8.4287>, rot=<0.0501 -0.0275 0.8926 -0.4472>}
 deform_glute_r {parent=deform_glute_base_r, loc=<0 0 0>, rot=<0 0 0 1>}
 thigh_r_null {parent=pelvis, type=Null, loc=<-2.5693 -0.0109 7.7501>, rot=<0.0572 -0.0012 0.9982 0.0155>}
]
variables [Sliding {type=bool, default=False}]
