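# Cosmetic AnimBP for the M_MED_MarkerDeer_Tin male body, adding secondary motion to hose and belt bones.
# It copies the parent mesh pose, constrains hose ends to spine, pelvis and thighs, then chains AnimDynamics, a sliding-aware Control Rig and belt AnimDynamics.
# A front-end idle curve blends DefaultSlot onto gun bones.
# The EventGraph also computes gravity, sim space and alpha variables, which the AnimGraph never reads.
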AnimBlueprint /BRCosmetics/Characters/Player/Male/Medium/Bodies/M_MED_MarkerDeer_Tin/Meshes/M_MED_MarkerDeer_Tin_AnimBP
target_skeleton=&/M_MED_MarkerDeer_Tin_Skeleton
graphs [
 AnimGraph {nodes [
   n1 = "Constraint - dyn_hose_upper14_end (P)"!(
    ComponentPose=Local To Component(LocalPose=Copy Pose From Mesh() {bUseAttachedParent=true})
    Alpha=1
    ConstraintWeights_0=1
    ConstraintWeights_1:float=0
   ) -> (Pose) {
    alphaBoolBlend {blendOption=Linear}
    boneToModify {boneName=dyn_hose_upper14_end}
    constraintSetup [{targetBone {boneName=thigh_twist_01_l}}]
    constraintWeights [1,0]
   }
   n2 = "Constraint - dyn_hose_upper13_end (P)"!(ComponentPose=@n1, Alpha=1, ConstraintWeights_0=1, ConstraintWeights_1:float=0) -> (Pose) {
    alphaBoolBlend {blendOption=Linear}
    boneToModify {boneName=dyn_hose_upper13_end}
    constraintSetup [{targetBone {boneName=thigh_twist_01_l}}]
    constraintWeights [1,0]
   }
   n3 = "Constraint - dyn_hose_upper12_end (P)"!(ComponentPose=@n2, Alpha=1, ConstraintWeights_0=1, ConstraintWeights_1:float=0) -> (Pose) {
    alphaBoolBlend {blendOption=Linear}
    boneToModify {boneName=dyn_hose_upper12_end}
    constraintSetup [{targetBone {boneName=thigh_twist_01_l}}]
    constraintWeights [1,0]
   }
   n4 = "Constraint - dyn_hose_upper11_end (P)"!(ComponentPose=@n3, Alpha=1, ConstraintWeights_0=1, ConstraintWeights_1:float=0) -> (Pose) {
    alphaBoolBlend {blendOption=Linear}
    boneToModify {boneName=dyn_hose_upper11_end}
    constraintSetup [{targetBone {boneName=thigh_twist_01_l}}]
    constraintWeights [1,0]
   }
   n5 = "Constraint - dyn_hose_upper10_end (P)"!(ComponentPose=@n4, Alpha=1, ConstraintWeights_0=1, ConstraintWeights_1:float=0) -> (Pose) {
    alphaBoolBlend {blendOption=Linear}
    boneToModify {boneName=dyn_hose_upper10_end}
    constraintSetup [{targetBone {boneName=thigh_twist_01_l}}]
    constraintWeights [1,0]
   }
   n6 = "Constraint - dyn_hose_upper9_end (PPP)"!(ComponentPose=@n5, Alpha=1, ConstraintWeights_0=0.7122, ConstraintWeights_1:float=0.263, ConstraintWeights_2:float=0.0248) -> (Pose) {
    alphaBoolBlend {blendOption=Linear}
    boneToModify {boneName=dyn_hose_upper9_end}
    constraintSetup [{targetBone {boneName=thigh_twist_01_l}}, {targetBone {boneName=pelvis}}, {targetBone {boneName=thigh_twist_01_r}}]
    constraintWeights [1,1,1]
   }
   n7 = "Constraint - dyn_hose_upper8_end (P)"!(ComponentPose=@n6, Alpha=1, ConstraintWeights_0=1, ConstraintWeights_1:float=0) -> (Pose) {
    alphaBoolBlend {blendOption=Linear}
    boneToModify {boneName=dyn_hose_upper8_end}
    constraintSetup [{targetBone {boneName=pelvis}}]
    constraintWeights [1,0]
   }
   n8 = "Constraint - dyn_hose_upper7_end (P)"!(ComponentPose=@n7, Alpha=1, ConstraintWeights_0=1, ConstraintWeights_1:float=0) -> (Pose) {
    alphaBoolBlend {blendOption=Linear}
    boneToModify {boneName=dyn_hose_upper7_end}
    constraintSetup [{targetBone {boneName=pelvis}}]
    constraintWeights [1,0]
   }
   n9 = "Constraint - dyn_hose_upper6_end (P)"!(ComponentPose=@n8, Alpha=1, ConstraintWeights_0=1, ConstraintWeights_1:float=0) -> (Pose) {
    alphaBoolBlend {blendOption=Linear}
    boneToModify {boneName=dyn_hose_upper6_end}
    constraintSetup [{targetBone {boneName=pelvis}}]
    constraintWeights [1,0]
   }
   n10 = "Constraint - dyn_hose_upper5_end (PPPP)"!(
    ComponentPose=@n9
    Alpha=1
    ConstraintWeights_0=0.5623
    ConstraintWeights_1:float=0.3038
    ConstraintWeights_2:float=0.1306
    ConstraintWeights_3:float=0.0033
   ) -> (Pose) {
    alphaBoolBlend {blendOption=Linear}
    boneToModify {boneName=dyn_hose_upper5_end}
    constraintSetup [{targetBone {boneName=spine_01}}, {targetBone {boneName=pelvis}}, {targetBone {boneName=spine_03}}, {targetBone {boneName=spine_02}}]
    constraintWeights [1,1,1,1]
   }
   n11 = "Constraint - dyn_hose_upper4_end (PP)"!(ComponentPose=@n10, Alpha=1, ConstraintWeights_0=0.6176, ConstraintWeights_1:float=0.3824) -> (Pose) {
    alphaBoolBlend {blendOption=Linear}
    boneToModify {boneName=dyn_hose_upper4_end}
    constraintSetup [{targetBone {boneName=spine_01}}, {targetBone {boneName=spine_03}}]
    constraintWeights [1,1]
   }
   n12 = "Constraint - dyn_hose_upper3_end (PP)"!(ComponentPose=@n11, Alpha=1, ConstraintWeights_0=0.8126, ConstraintWeights_1:float=0.1874) -> (Pose) {
    alphaBoolBlend {blendOption=Linear}
    boneToModify {boneName=dyn_hose_upper3_end}
    constraintSetup [{targetBone {boneName=spine_03}}, {targetBone {boneName=spine_01}}]
    constraintWeights [1,1]
   }
   n13 = "Constraint - dyn_hose_upper2_end (PP)"!(ComponentPose=@n12, Alpha=1, ConstraintWeights_0=0.7372, ConstraintWeights_1:float=0.2628) -> (Pose) {
    alphaBoolBlend {blendOption=Linear}
    boneToModify {boneName=dyn_hose_upper2_end}
    constraintSetup [{targetBone {boneName=spine_03}}, {targetBone {boneName=spine_04}}]
    constraintWeights [1,1]
   }
   n14 = "Constraint - dyn_hose_upper1_end (P)"!(ComponentPose=@n13, Alpha=1, ConstraintWeights_0=1, ConstraintWeights_1:float=0) -> (Pose) {
    alphaBoolBlend {blendOption=Linear}
    boneToModify {boneName=dyn_hose_upper1_end}
    constraintSetup [{targetBone {boneName=spine_04}}]
    constraintWeights [1,0]
   }
   n15 = "Constraint - dyn_hose_upper_end (P)"!(ComponentPose=@n14, Alpha=1, ConstraintWeights_0=1, ConstraintWeights_1:float=0) -> (Pose) {
    alphaBoolBlend {blendOption=Linear}
    boneToModify {boneName=dyn_hose_upper_end}
    constraintSetup [{targetBone {boneName=spine_04}}]
    constraintWeights [1,0]
   }
   n16 = Anim Dynamics - Bone: dyn_hose_upper8_end!(SimulationSpace:AnimPhysSimSpaceType Enum, ComponentPose=@n15, Alpha=1, ExternalForce="0,0,0") -> (Pose) {
    angularDampingOverride=1
    angularSpringConstant=20000
    bAngularSpring=true
    bLinearSpring=true
    bOverrideAngularDamping=true
    bOverrideLinearDamping=true
    bUseGravityOverride=true
    gravityScale=0.35
    linearDampingOverride=1
    linearSpringConstant=100
    simulationSpace=RootRelative
    alphaBoolBlend {blendOption=Linear}
    boundBone {boneName=dyn_hose_upper8_end}
    componentAppliedLinearAccClamp=<1500 1500 1500>
    componentLinearAccScale=<1 1 1>
    componentLinearVelScale=<1 1 1>
    gravityOverride {z=-980}
    physicsBodyDefinitions [{
      boundBone {boneName=dyn_hose_upper8_end}
      boxExtents=<30 30 30>
      constraintSetup {angularLimitsMax=<0 0 0>, angularLimitsMin=<0 0 0>, angularTarget=<1 0 0>, linearAxesMax=<2 0 2>, linearAxesMin=<-2 0 -2>}
      localJointOffset=<4 0 0>
     }]
   }
   n17 = Anim Dynamics - Bone: dyn_hose_upper4_end!(SimulationSpace:AnimPhysSimSpaceType Enum, ComponentPose=@n16, Alpha=1, ExternalForce="0,0,0") -> (Pose) {
    angularDampingOverride=1
    angularSpringConstant=20000
    bAngularSpring=true
    bLinearSpring=true
    bOverrideAngularDamping=true
    bOverrideLinearDamping=true
    bUseGravityOverride=true
    gravityScale=0.35
    linearDampingOverride=1
    linearSpringConstant=100
    simulationSpace=RootRelative
    alphaBoolBlend {blendOption=Linear}
    boundBone {boneName=dyn_hose_upper4_end}
    componentAppliedLinearAccClamp=<1500 1500 1500>
    componentLinearAccScale=<1 1 1>
    componentLinearVelScale=<1 1 1>
    gravityOverride {z=-980}
    physicsBodyDefinitions [{
      boundBone {boneName=dyn_hose_upper4_end}
      boxExtents=<30 30 30>
      constraintSetup {angularLimitsMax=<0 0 0>, angularLimitsMin=<0 0 0>, angularTarget=<1 0 0>, linearAxesMax=<2 0 2>, linearAxesMin=<-2 0 -2>}
      localJointOffset=<4 0 0>
     }]
   }
   n18 = Control Rig!(Source=Component To Local(ComponentPose=@n17), Alpha, Sliding:bool=Get bIsSliding()) -> (Pose) {
    alphaBoolBlend {blendOption=Linear}
    controlRigAssetReference {blueprintRigClass=&/M_MED_MarkerDeer_Tin_CtrlRig.M_MED_MarkerDeer_Tin_CtrlRig_C}
   }
   n19 = Anim Dynamics - Bone: dyn_belt_c_01!(
    SimulationSpace:AnimPhysSimSpaceType Enum
    ComponentPose=Local To Component(LocalPose=@n18)
    Alpha=1
    GravityOverride:Vector="0,0,-980"
    ExternalForce="0,0,0"
   ) -> (Pose) {
    angularDampingOverride=1
    angularSpringConstant=50
    bAngularSpring=true
    bOverrideAngularDamping=true
    bUseGravityOverride=true
    gravityScale=0.35
    simulationSpace=RootRelative
    alphaBoolBlend {blendOption=Linear}
    boundBone {boneName=dyn_belt_c_01}
    componentAppliedLinearAccClamp=<1500 1500 1500>
    componentLinearAccScale=<0.4 0.7 0.7>
    componentLinearVelScale=<0.4 0.7 0.7>
    gravityOverride {z=-980}
    physicsBodyDefinitions [{
      boundBone {boneName=dyn_belt_c_01}
      boxExtents=<17 17 17>
      constraintSetup {angularLimitsMax=<1 10 3>, angularLimitsMin=<-1 0 -3>, angularTarget=<1 0 0>}
      localJointOffset=<3 0 0>
     }]
   }
   n20 = Anim Dynamics - Bone: dyn_belt_c_02!(SimulationSpace:AnimPhysSimSpaceType Enum, ComponentPose=@n19, Alpha=1, GravityOverride:Vector="0,0,-1000", ExternalForce="0,0,0") -> (Pose) {
    angularDampingOverride=1
    angularSpringConstant=50
    bAngularSpring=true
    bOverrideAngularDamping=true
    bUseGravityOverride=true
    gravityScale=0.35
    simulationSpace=RootRelative
    alphaBoolBlend {blendOption=Linear}
    boundBone {boneName=dyn_belt_c_02}
    componentAppliedLinearAccClamp=<1500 1500 1500>
    componentLinearAccScale=<0.4 0.7 0.7>
    componentLinearVelScale=<0.4 0.7 0.7>
    gravityOverride {z=-1000}
    physicsBodyDefinitions [{
      boundBone {boneName=dyn_belt_c_02}
      boxExtents=<17 17 17>
      constraintSetup {angularLimitsMax=<1 10 3>, angularLimitsMin=<-1 0 -3>, angularTarget=<1 0 0>}
      localJointOffset=<3 0 0>
     }]
   }
   n21 = Anim Dynamics - Bone: dyn_belt_c_03!(SimulationSpace:AnimPhysSimSpaceType Enum, ComponentPose=@n20, Alpha=1, GravityOverride:Vector="0,0,-1000", ExternalForce="0,0,0") -> (Pose) {
    angularDampingOverride=1
    angularSpringConstant=50
    bAngularSpring=true
    bOverrideAngularDamping=true
    bUseGravityOverride=true
    gravityScale=0.35
    simulationSpace=RootRelative
    alphaBoolBlend {blendOption=Linear}
    boundBone {boneName=dyn_belt_c_03}
    componentAppliedLinearAccClamp=<1500 1500 1500>
    componentLinearAccScale=<0.4 0.7 0.7>
    componentLinearVelScale=<0.4 0.7 0.7>
    gravityOverride {z=-1000}
    physicsBodyDefinitions [{
      boundBone {boneName=dyn_belt_c_03}
      boxExtents=<17 17 17>
      constraintSetup {angularLimitsMax=<1 10 3>, angularLimitsMin=<-1 0 -3>, angularTarget=<1 0 0>}
      localJointOffset=<3 0 0>
     }]
   }
   n22 = Anim Dynamics - Bone: dyn_belt_l_01!(SimulationSpace:AnimPhysSimSpaceType Enum, ComponentPose=@n21, Alpha=1, GravityOverride:Vector="0,0,-1000", ExternalForce="0,0,0") -> (Pose) {
    angularDampingOverride=0.9999
    bAngularSpring=true
    bOverrideAngularDamping=true
    bUseGravityOverride=true
    gravityScale=0.35
    simulationSpace=RootRelative
    alphaBoolBlend {blendOption=Linear}
    boundBone {boneName=dyn_belt_l_01}
    componentAppliedLinearAccClamp=<1500 1500 1500>
    componentLinearAccScale=<0.7 0.5 0.5>
    componentLinearVelScale=<0.7 0.5 0.5>
    gravityOverride {z=-1000}
    physicsBodyDefinitions [{
      boundBone {boneName=dyn_belt_l_01}
      boxExtents=<15 15 15>
      constraintSetup {angularLimitsMax=<1 5 0>, angularLimitsMin=<-1 -5 -10>, angularTarget=<1 0 0>}
      localJointOffset=<5 0 0>
     }]
   }
   n23 = Anim Dynamics - Bone: dyn_belt_l_02!(SimulationSpace:AnimPhysSimSpaceType Enum, ComponentPose=@n22, Alpha=1, GravityOverride:Vector="0,0,-1000", ExternalForce="0,0,0") -> (Pose) {
    angularDampingOverride=0.9999
    bAngularSpring=true
    bOverrideAngularDamping=true
    bUseGravityOverride=true
    gravityScale=0.35
    simulationSpace=RootRelative
    alphaBoolBlend {blendOption=Linear}
    boundBone {boneName=dyn_belt_l_02}
    componentAppliedLinearAccClamp=<1500 1500 1500>
    componentLinearAccScale=<0.7 0.5 0.5>
    componentLinearVelScale=<0.7 0.5 0.5>
    gravityOverride {z=-1000}
    physicsBodyDefinitions [{
      boundBone {boneName=dyn_belt_l_02}
      boxExtents=<15 15 15>
      constraintSetup {angularLimitsMax=<1 5 0>, angularLimitsMin=<-1 -5 -10>, angularTarget=<1 0 0>}
      localJointOffset=<5 0 0>
     }]
   }
   n24 = Anim Dynamics - Bone: dyn_belt_r_01!(SimulationSpace:AnimPhysSimSpaceType Enum, ComponentPose=@n23, Alpha=1, GravityOverride:Vector="0,0,-1000", ExternalForce="0,0,0") -> (Pose) {
    angularDampingOverride=0.9999
    bAngularSpring=true
    bOverrideAngularDamping=true
    bUseGravityOverride=true
    gravityScale=0.35
    simulationSpace=RootRelative
    alphaBoolBlend {blendOption=Linear}
    boundBone {boneName=dyn_belt_r_01}
    componentAppliedLinearAccClamp=<1500 1500 1500>
    componentLinearAccScale=<0.7 0.5 0.5>
    componentLinearVelScale=<0.7 0.5 0.5>
    gravityOverride {z=-1000}
    physicsBodyDefinitions [{
      boundBone {boneName=dyn_belt_r_01}
      boxExtents=<15 15 15>
      constraintSetup {angularLimitsMax=<1 5 0>, angularLimitsMin=<-1 -5 -10>, angularTarget=<1 0 0>}
      localJointOffset=<-5 0 0>
     }]
   }
   n25 = Anim Dynamics - Bone: dyn_belt_r_02!(SimulationSpace:AnimPhysSimSpaceType Enum, ComponentPose=@n24, Alpha=1, GravityOverride:Vector="0,0,-1000", ExternalForce="0,0,0") -> (Pose) {
    angularDampingOverride=0.9999
    bAngularSpring=true
    bOverrideAngularDamping=true
    bUseGravityOverride=true
    gravityScale=0.35
    simulationSpace=RootRelative
    alphaBoolBlend {blendOption=Linear}
    boundBone {boneName=dyn_belt_r_02}
    componentAppliedLinearAccClamp=<1500 1500 1500>
    componentLinearAccScale=<0.7 0.5 0.5>
    componentLinearVelScale=<0.7 0.5 0.5>
    gravityOverride {z=-1000}
    physicsBodyDefinitions [{
      boundBone {boneName=dyn_belt_r_02}
      boxExtents=<15 15 15>
      constraintSetup {angularLimitsMax=<1 5 0>, angularLimitsMin=<-1 -5 -10>, angularTarget=<1 0 0>}
      localJointOffset=<-5 0 0>
     }]
   }
   Output Pose(Result=Layered blend per bone(
     BasePose=Component To Local(ComponentPose=@n25)
     BlendPoses_0=Slot 'DefaultSlot'()
     BlendWeights_0=Get CustomFELobbyIdleFloat() -> (CustomFELobbyIdleFloat:float)
    ) {layerSetup [{branchFilters [{boneName=dyn_gun_02}, {boneName=dyn_gun_01}]}]}) {layerGroup=None}
  ]}
 EventGraph {nodes [
   Event BlueprintUpdateAnimation()
   > n1 = Sequence!(execute) -> (then_0, then_1, then_2:Exec, then_3:Exec, then_4:Exec)
   > n2 = Gravity Override State Function DEPRECATED(CharacterPartAnimInstance=Self-Reference(), GameTimeInSeconds=Get Delta Time X(), GravityOverrideParams=Get Gravity Override Params())
   > Set Gravity Override(Gravity Override=@n2)
   Set simSpace(execute=@n1.then_1, simSpace:AnimPhysSimSpaceType Enum) -> (Output_Get:AnimPhysSimSpaceType Enum)
   > n3 = Branch(Condition=Get bIsInFrontEnd())
   > Set simSpace(simSpace=World)
   Set simSpace(execute=@n3.else)
   n4 = AND Boolean(A=Get bIsCrouching(), B=NOT Boolean(A=Get bIsCrouchSprinting()))
   n5 = Get bIsSkyDiving()
   n6 = Get bIsJumping()
   n7 = Get bIsFalling()
   n8 = Get bIsSliding()
   Set Constrain Belt Alpha(
    execute=@n1.then_2
    Constrain Belt Alpha:double="To Float (Boolean)"(InBool=OR Boolean!(A=@n4, B=@n8, C:bool=@n6, D:bool=@n7, E:bool=@n5) -> (ReturnValue))
   ) -> (Output_Get:double)
   Set frontEndAlpha(execute=@n1.then_3, frontEndAlpha:double=SelectFloat(A=0.2, B=1, bPickA=Get bIsInFrontEnd())) -> (Output_Get:double)
   Branch(execute=@n1.then_4, Condition=Get bIsInFrontEnd())
   > Set CustomFELobbyIdleFloat(CustomFELobbyIdleFloat:float=GetCurveValue(self=GetOwnerFrontEndAnimInstance(), CurveName=FrontEnd_Idle_Bypass)) -> (Output_Get:float)
   Get bIsOnGrindRail()
   Get bIsPlayingEmote()
   Set BodyMesh(BodyMesh=GetOwnerSkeletalMeshComponent())
  ]}
]
variables [
 BodyMesh {type=SkeletalMeshComponent}
 Constrain Belt Alpha {type=number}
 CustomFELobbyIdleFloat {type=number}
 Delta Time X {type=number}
 frontEndAlpha {type=number}
 Gravity Override {type=Vector}
 Gravity Override Params {type=GravityOverrideParamsStruct}
 simSpace {type=AnimPhysSimSpaceType}
]
